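# Physics asset for the collar of the F_MED_Pliant female medium player body, set up for secondary motion of the back collar.
# Left, center and right back collar capsules are each constrained to a tiny matching MSTR anchor body with identical limited swing and tight twist.
# A head capsule, likely a collider, and front left and right MSTR bodies carry no constraints.
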PhysicsAsset /BRCosmetics/Characters/Player/Female/Medium/Bodies/F_MED_Pliant/Meshes/F_MED_Pliant_Collar_Physics
bodies {
 dyn_collar_bK_l [{shapeType=Capsule, length=3.3138, radius=5.4263, center=<3.7875 -2.2 -2.2>, rotation=<rot 84.0879 179.9999 179.9999>}]
 dyn_collar_bK_l_MSTR [{shapeType=Capsule, length=0.505, radius=0.505, rotation=<rot 0 -90 0>}]
 dyn_collar_bK_r [{shapeType=Capsule, length=4.2351, radius=4.512, center=<5.3633 2 -1>, rotation=<rot 86.2407 -179.9995 -179.9995>}]
 dyn_collar_bK_r_MSTR [{shapeType=Capsule, length=0.505, radius=0.505, rotation=<rot 0 -90 0>}]
 dyn_collar_bk [{shapeType=Capsule, length=4.9566, radius=6.1178, center=<3.2658 0 -4.5>, rotation=<rot 90 0 0>}]
 dyn_collar_bk_MSTR [{shapeType=Capsule, length=0.505, radius=0.505, rotation=<rot 0 -90 0>}]
 dyn_collar_l_MSTR [{shapeType=Capsule, length=0.505, radius=0.505, rotation=<rot 0 -90 0>}]
 dyn_collar_r_MSTR [{shapeType=Capsule, length=0.505, radius=0.505, rotation=<rot 0 -90 0>}]
 head [{shapeType=Capsule, length=6.3314, radius=7.5339, center=<1.1447 -1.3673 0>, rotation=<rot 90 53.9726 63.9726>}]
}
constraints [
 {bone1Name=dyn_collar_bK_l, bone2Name=dyn_collar_bK_l_MSTR, swing1LimitDegrees=15, swing2LimitDegrees=25, twistLimitDegrees=3}
 {bone1Name=dyn_collar_bK_r, bone2Name=dyn_collar_bK_r_MSTR, swing1LimitDegrees=15, swing2LimitDegrees=25, twistLimitDegrees=3}
 {bone1Name=dyn_collar_bk, bone2Name=dyn_collar_bk_MSTR, swing1LimitDegrees=15, swing2LimitDegrees=25, twistLimitDegrees=3}
]
