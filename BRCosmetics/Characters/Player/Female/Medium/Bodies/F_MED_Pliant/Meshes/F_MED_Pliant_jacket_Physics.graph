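# Physics asset for the F_MED_Pliant female body's jacket, defining collision bodies and joints for its dynamic bones.
# Seven three-segment chains around the back, sides and front each hang from a small master body constrained to spine_03, alongside pelvis, spine and thigh capsules.
# Chain joints have tight angular limits, looser at the front.
# Front-left uses two boxes among capsules and has an angularly free cross-link to the left-side chain.
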PhysicsAsset /BRCosmetics/Characters/Player/Female/Medium/Bodies/F_MED_Pliant/Meshes/F_MED_Pliant_jacket_Physics
bodies {
 dyn_jacket_bk_01 [{shapeType=Capsule, length=4.9231, radius=3.8992, center=<4.9576 0 2>, rotation=<rot 0 0 90>}]
 dyn_jacket_bk_02 [{shapeType=Capsule, length=3.3924, radius=4.4263, center=<3.9964 0 2.5>, rotation=<rot 0 0 90>}]
 dyn_jacket_bk_03 [{shapeType=Capsule, length=5.4787, radius=5.0797, center=<4.6698 0 3>, rotation=<rot 0 0 90>}]
 dyn_jacket_bk_MSTR [{shapeType=Capsule, length=0.505, radius=0.505}]
 dyn_jacket_bk_l_01 [{shapeType=Capsule, length=2.3334, radius=4.5218, center=<4.781 0 2.2>, rotation=<rot 0 0 -90>}]
 dyn_jacket_bk_l_02 [{shapeType=Capsule, length=3.7637, radius=4.1932, center=<2.9679 0 3.5>, rotation=<rot 0 0 -94.9999>}]
 dyn_jacket_bk_l_03 [{shapeType=Capsule, length=6.4617, radius=4.4024, center=<4.0233 0 3.5>, rotation=<rot 0 0 -95.0002>}]
 dyn_jacket_bk_l_MSTR [{shapeType=Capsule, length=0.505, radius=0.505}]
 dyn_jacket_bk_r_01 [{shapeType=Capsule, length=3.7123, radius=4.1599, center=<4.7795 0 2.5>, rotation=<rot 0 0 90>}]
 dyn_jacket_bk_r_02 [{shapeType=Capsule, length=5.2435, radius=4.2304, center=<4.2014 0 3.3>, rotation=<rot 0 0 94.9999>}]
 dyn_jacket_bk_r_03 [{shapeType=Capsule, length=10.2585, radius=4.7715, center=<4.2011 -1 3.5>, rotation=<rot 0 0 -85.5412>}]
 dyn_jacket_bk_r_MSTR [{shapeType=Capsule, length=0.505, radius=0.505}]
 dyn_jacket_fr_l_01 [{
   shapeType=Box
   extentX=8.7477
   extentY=9.6988
   extentZ=3.1929
   shapeName=dyn_jacket_fr_l_01_box
   center=<3 -2 0>
   rotation=<rot 3.1215 -89.9998 0>
  }]
 dyn_jacket_fr_l_02 [{
   shapeType=Box
   extentX=10.822
   extentY=3.6494
   extentZ=11.0854
   shapeName=dyn_jacket_fr_l_02_box
   center=<3.2832 -1.5 0>
   rotation=<rot 0 0 85.869>
  }]
 dyn_jacket_fr_l_03 [{shapeType=Capsule, length=8.3406, radius=3.8941, shapeName=dyn_jacket_fr_l_03_capsule, center=<4.6376 0 2>, rotation=<rot 90 0 -65>}]
 dyn_jacket_fr_l_MSTR [{shapeType=Capsule, length=0.505, radius=0.505}]
 dyn_jacket_fr_r_01 [{shapeType=Capsule, length=6.1038, radius=4.867, center=<4.9485 3 3>, rotation=<rot 0 0 -90>}]
 dyn_jacket_fr_r_02 [{shapeType=Capsule, length=6.2225, radius=5.348, center=<4.9264 3 3>, rotation=<rot 1.7083 4.7 -109.9303>}]
 dyn_jacket_fr_r_03 [{shapeType=Capsule, length=8.5535, radius=4.8459, center=<4.509 1.2 1.5>, rotation=<rot 0 29.9997 64.9991>}]
 dyn_jacket_fr_r_MSTR [{shapeType=Capsule, length=0.505, radius=0.505}]
 dyn_jacket_l_01 [{shapeType=Capsule, length=2.6176, radius=4.904, center=<4.6406 0 2.5>, rotation=<rot 0 0 -87.2508>}]
 dyn_jacket_l_02 [{shapeType=Capsule, length=4.4016, radius=4.6678, center=<3.9272 0 3>, rotation=<rot 0.0002 0 -81.5863>}]
 dyn_jacket_l_03 [{shapeType=Capsule, length=5.6427, radius=4.9843, center=<3.1553 0 3>, rotation=<rot 0 0 -81.4895>}]
 dyn_jacket_l_MSTR [{shapeType=Capsule, length=0.505, radius=0.505}]
 dyn_jacket_r_01 [{shapeType=Capsule, length=4.0222, radius=4.1587, center=<4.9304 0 3>, rotation=<rot 0 0 90>}]
 dyn_jacket_r_02 [{shapeType=Capsule, length=8.2371, radius=4.0177, center=<4.121 0 3>, rotation=<rot 0 0 90>}]
 dyn_jacket_r_03 [{shapeType=Capsule, length=12.1204, radius=4.7613, center=<3.1866 0 2>, rotation=<rot 0 0 -105.0003>}]
 dyn_jacket_r_MSTR [{shapeType=Capsule, length=0.505, radius=0.505}]
 pelvis [{shapeType=Capsule, length=11.0873, radius=10.2191, center=<-3.4098 -2.555 0>}]
 spine_01 [{shapeType=Capsule, length=6.1174, radius=9.8109, center=<-0.8005 -0.0001 0>}]
 spine_02 [{shapeType=Capsule, length=3.9632, radius=9.0354, center=<-2.0322 0.5442 0>}]
 spine_03 [{shapeType=Capsule, length=2.3618, radius=8.435, center=<-1.56 1.5943 0>}]
 thigh_l [
  {shapeType=Capsule, length=23.8982, radius=9.0492, center=<-16.9257 -1.2024 1.7626>, rotation=<rot 83.921 178.7493 177.6626>}
  {
   shapeType=Capsule
   length=14.8401
   radius=3.9074
   shapeName=thigh_l_capsule
   center=<-9.9062 3.2305 -3.7859>
   rotation=<rot 53.0515 -105.9833 -109.7167>
  }
 ]
 thigh_r [{shapeType=Capsule, length=22.7916, radius=10.1558, center=<16.0844 1.6939 -1.1767>, rotation=<rot 85.7425 -179.5688 179.9999>}]
}
constraints [
 {bone1Name=thigh_l, bone2Name=pelvis, swing1LimitDegrees=45, swing2LimitDegrees=45, twistLimitDegrees=45}
 {bone1Name=thigh_r, bone2Name=pelvis, swing1LimitDegrees=45, swing2LimitDegrees=45, twistLimitDegrees=45}
 {
  bone1Name=spine_01
  bone2Name=pelvis
  swing1LimitDegrees=45
  swing1Motion=Free
  swing2LimitDegrees=45
  swing2Motion=Free
  twistLimitDegrees=45
  twistMotion=Free
 }
 {
  bone1Name=spine_02
  bone2Name=spine_01
  swing1LimitDegrees=45
  swing1Motion=Free
  swing2LimitDegrees=45
  swing2Motion=Free
  twistLimitDegrees=45
  twistMotion=Free
 }
 {
  bone1Name=spine_03
  bone2Name=spine_02
  swing1LimitDegrees=45
  swing1Motion=Free
  swing2LimitDegrees=45
  swing2Motion=Free
  twistLimitDegrees=45
  twistMotion=Free
 }
 {
  bone1Name=dyn_jacket_bk_MSTR
  bone2Name=spine_03
  swing1LimitDegrees=45
  swing1Motion=Free
  swing2LimitDegrees=45
  swing2Motion=Free
  twistLimitDegrees=45
  twistMotion=Free
 }
 {bone1Name=dyn_jacket_bk_01, bone2Name=dyn_jacket_bk_MSTR, swing1LimitDegrees=10, swing2LimitDegrees=20, twistLimitDegrees=5}
 {bone1Name=dyn_jacket_bk_02, bone2Name=dyn_jacket_bk_01, swing1LimitDegrees=10, swing2LimitDegrees=35, twistLimitDegrees=3}
 {bone1Name=dyn_jacket_bk_03, bone2Name=dyn_jacket_bk_02, swing1LimitDegrees=10, swing2LimitDegrees=35, twistLimitDegrees=3}
 {
  bone1Name=dyn_jacket_bk_r_MSTR
  bone2Name=spine_03
  swing1LimitDegrees=45
  swing1Motion=Free
  swing2LimitDegrees=45
  swing2Motion=Free
  twistLimitDegrees=45
  twistMotion=Free
 }
 {bone1Name=dyn_jacket_bk_r_01, bone2Name=dyn_jacket_bk_r_MSTR, swing1LimitDegrees=10, swing2LimitDegrees=20, twistLimitDegrees=5}
 {bone1Name=dyn_jacket_bk_r_02, bone2Name=dyn_jacket_bk_r_01, swing1LimitDegrees=10, swing2LimitDegrees=35, twistLimitDegrees=3}
 {bone1Name=dyn_jacket_bk_r_03, bone2Name=dyn_jacket_bk_r_02, swing1LimitDegrees=10, swing2LimitDegrees=35, twistLimitDegrees=3}
 {
  bone1Name=dyn_jacket_r_MSTR
  bone2Name=spine_03
  swing1LimitDegrees=45
  swing1Motion=Free
  swing2LimitDegrees=45
  swing2Motion=Free
  twistLimitDegrees=45
  twistMotion=Free
 }
 {bone1Name=dyn_jacket_r_01, bone2Name=dyn_jacket_r_MSTR, swing1LimitDegrees=10, swing2LimitDegrees=20, twistLimitDegrees=5}
 {bone1Name=dyn_jacket_r_02, bone2Name=dyn_jacket_r_01, swing1LimitDegrees=10, swing2LimitDegrees=35, twistLimitDegrees=3}
 {bone1Name=dyn_jacket_r_03, bone2Name=dyn_jacket_r_02, swing1LimitDegrees=10, swing2LimitDegrees=35, twistLimitDegrees=3}
 {
  bone1Name=dyn_jacket_fr_r_MSTR
  bone2Name=spine_03
  swing1LimitDegrees=45
  swing1Motion=Free
  swing2LimitDegrees=45
  swing2Motion=Free
  twistLimitDegrees=45
  twistMotion=Free
 }
 {bone1Name=dyn_jacket_fr_r_01, bone2Name=dyn_jacket_fr_r_MSTR, swing1LimitDegrees=15, swing2LimitDegrees=20, twistLimitDegrees=5}
 {bone1Name=dyn_jacket_fr_r_02, bone2Name=dyn_jacket_fr_r_01, swing1LimitDegrees=35, swing2LimitDegrees=35, twistLimitDegrees=15}
 {bone1Name=dyn_jacket_fr_r_03, bone2Name=dyn_jacket_fr_r_02, swing1LimitDegrees=35, swing2LimitDegrees=35, twistLimitDegrees=15}
 {
  bone1Name=dyn_jacket_fr_l_MSTR
  bone2Name=spine_03
  swing1LimitDegrees=45
  swing1Motion=Free
  swing2LimitDegrees=45
  swing2Motion=Free
  twistLimitDegrees=45
  twistMotion=Free
 }
 {bone1Name=dyn_jacket_fr_l_01, bone2Name=dyn_jacket_fr_l_MSTR, swing1LimitDegrees=15, swing2LimitDegrees=20, twistLimitDegrees=5}
 {bone1Name=dyn_jacket_fr_l_02, bone2Name=dyn_jacket_fr_l_01, swing1LimitDegrees=35, swing2LimitDegrees=35, twistLimitDegrees=15}
 {
  bone1Name=dyn_jacket_l_MSTR
  bone2Name=spine_03
  swing1LimitDegrees=45
  swing1Motion=Free
  swing2LimitDegrees=45
  swing2Motion=Free
  twistLimitDegrees=45
  twistMotion=Free
 }
 {bone1Name=dyn_jacket_l_01, bone2Name=dyn_jacket_l_MSTR, swing1LimitDegrees=10, swing2LimitDegrees=20, twistLimitDegrees=5}
 {bone1Name=dyn_jacket_l_02, bone2Name=dyn_jacket_l_01, swing1LimitDegrees=10, swing2LimitDegrees=35, twistLimitDegrees=3}
 {bone1Name=dyn_jacket_l_03, bone2Name=dyn_jacket_l_02, swing1LimitDegrees=10, swing2LimitDegrees=35, twistLimitDegrees=3}
 {
  bone1Name=dyn_jacket_bk_l_MSTR
  bone2Name=spine_03
  swing1LimitDegrees=45
  swing1Motion=Free
  swing2LimitDegrees=45
  swing2Motion=Free
  twistLimitDegrees=45
  twistMotion=Free
 }
 {bone1Name=dyn_jacket_bk_l_01, bone2Name=dyn_jacket_bk_l_MSTR, swing1LimitDegrees=10, swing2LimitDegrees=20, twistLimitDegrees=5}
 {bone1Name=dyn_jacket_bk_l_02, bone2Name=dyn_jacket_bk_l_01, swing1LimitDegrees=10, swing2LimitDegrees=35, twistLimitDegrees=3}
 {bone1Name=dyn_jacket_bk_l_03, bone2Name=dyn_jacket_bk_l_02, swing1LimitDegrees=10, swing2LimitDegrees=35, twistLimitDegrees=3}
 {bone1Name=dyn_jacket_fr_l_03, bone2Name=dyn_jacket_fr_l_02, swing1LimitDegrees=35, swing2LimitDegrees=35, twistLimitDegrees=15}
 {
  bone1Name=dyn_jacket_fr_l_02
  bone2Name=dyn_jacket_l_02
  swing1LimitDegrees=45
  swing1Motion=Free
  swing2LimitDegrees=45
  swing2Motion=Free
  twistLimitDegrees=45
  twistMotion=Free
 }
]
