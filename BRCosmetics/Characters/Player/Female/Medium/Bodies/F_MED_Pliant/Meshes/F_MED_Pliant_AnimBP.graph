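# Post-process style AnimBlueprint for the F_MED_Pliant female character body, adding secondary motion to jacket, belt and collar.
# It copies the parent mesh pose through a deform Control Rig, then drives helper bones with weighted constraints, look-ats and belt AnimDynamics.
# Cached poses feed crouch-dependent jacket corrections driven by thigh and spine values.
# Jacket simulation switches between RigidBody (separate front-end setup) and cheaper AnimDynamics chains, followed by collar RigidBody.
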AnimBlueprint /BRCosmetics/Characters/Player/Female/Medium/Bodies/F_MED_Pliant/Meshes/F_MED_Pliant_AnimBP
target_skeleton=/Game/Characters/Player/Female/Medium/Bodies/F_MED_Pliant/Meshes/F_MED_Pliant_Skeleton
graphs [
 AnimGraph {nodes [
   Anim Dynamics - Bone: dyn_belt_loop_01!(ComponentPose, bAlphaBoolEnabled:bool=Get bIsCrouching(), ExternalForce="0,0,0") -> (Pose) {
    alphaInputType=Bool
    angularDampingOverride=0.9999
    bAngularSpring=true
    bGravityOverrideInSimSpace=true
    bOverrideAngularDamping=true
    gravityScale=1.2
    simulationSpace=RootRelative
    alphaBoolBlend {blendInTime=0.1, blendOption=Linear, blendOutTime=0.1}
    boundBone {boneName=dyn_belt_loop_01}
    componentAppliedLinearAccClamp=<3500 3500 4300>
    componentLinearAccScale=<2 2 2>
    componentLinearVelScale=<2 2 2>
    physicsBodyDefinitions [{
      sphereCollisionRadius=0
      boundBone {boneName=dyn_belt_loop_01}
      boxExtents=<9 18 18>
      constraintSetup {angularLimitsMax=<45 0 60>, angularLimitsMin=<-45 -80 -60>, angularTarget=<0 0 0>}
      localJointOffset=<2 0 0>
     }]
    relativeSpaceBone {boneName=root}
   }
   Anim Dynamics - Bone: dyn_belt_loop_01(bAlphaBoolEnabled=Get bIsCrouching(), ExternalForce="0,0,0") {
    alphaInputType=Bool
    angularDampingOverride=0.9999
    bAngularSpring=true
    bGravityOverrideInSimSpace=true
    bOverrideAngularDamping=true
    gravityScale=1.2
    simulationSpace=RootRelative
    alphaBoolBlend {blendInTime=0.1, blendOption=Linear, blendOutTime=0.1}
    boundBone {boneName=dyn_belt_loop_01}
    componentAppliedLinearAccClamp=<3500 3500 4300>
    componentLinearAccScale=<2 2 2>
    componentLinearVelScale=<2 2 2>
    physicsBodyDefinitions [{
      sphereCollisionRadius=0
      boundBone {boneName=dyn_belt_loop_01}
      boxExtents=<9 18 18>
      constraintSetup {angularLimitsMax=<45 0 60>, angularLimitsMin=<-45 -80 -60>, angularTarget=<0 0 0>}
      localJointOffset=<2 0 0>
     }]
    relativeSpaceBone {boneName=root}
   }
   n1 = RigidBody!(
    ComponentPose=Local To Component(LocalPose=Use cached pose 'presim'())
    bAlphaBoolEnabled:bool
    OverrideWorldGravity:Vector=Get GravityOverride()
    ExternalForce
    ComponentLinearAccScale:Vector=Get RigidsAcc() -> (RigidsAcc:Vector)
    ComponentLinearVelScale:Vector=Get RigidsVel() -> (RigidsVel:Vector)
   ) -> (Pose) {
    alphaInputType=Bool
    bClampLinearTranslationLimitToRefPose=true
    bOverrideWorldGravity=true
    overridePhysicsAsset=&/F_MED_Pliant_jacket_Physics
    simulationSpace=BaseBoneSpace
    alphaBoolBlend {blendInTime=0.7, blendOption=QuarticInOut, blendOutTime=0.1}
    baseBoneRef {boneName=pelvis}
    componentAppliedLinearAccClamp=<2000 2000 2000>
    componentLinearAccScale=<0.3 0.2 0.2>
    componentLinearVelScale=<0.25 0.1 0.1>
    overrideWorldGravity {z=-980}
    simSpaceSettings {
     maxAngularAcceleration=1000
     maxAngularVelocity=3
     maxLinearAcceleration=700
     maxLinearVelocity=500
     velocityScaleZ=0
     worldAlpha=1
     externalLinearDragV=<1 1>
    }
   }
   n2 = RigidBody!(ComponentPose=Local To Component(LocalPose=Use cached pose 'presim'()), bAlphaBoolEnabled, ExternalForce) -> (Pose) {
    alphaInputType=Bool
    bClampLinearTranslationLimitToRefPose=true
    bOverrideWorldGravity=true
    overridePhysicsAsset=&/F_MED_Pliant_FrontEnd_Physics
    simulationSpace=BaseBoneSpace
    alphaBoolBlend {blendInTime=0.7, blendOption=QuarticInOut, blendOutTime=0.1}
    baseBoneRef {boneName=pelvis}
    componentLinearAccScale=<0.5 0.5 0.5>
    componentLinearVelScale=<0.5 0.5 0.5>
    overrideWorldGravity {z=-720}
    simSpaceSettings {
     maxAngularAcceleration=1000
     maxAngularVelocity=3
     maxLinearAcceleration=700
     maxLinearVelocity=500
     velocityScaleZ=0
     worldAlpha=1
     externalLinearDragV=<1 1>
    }
   }
   n3 = Blend Poses by bool(BlendPose_0=Component To Local(ComponentPose=@n2), BlendPose_1=Component To Local(ComponentPose=@n1), bActiveValue=Get bIsInFrontEnd())
   n4 = Anim Dynamics - Bone: dyn_jacket_bk_l_01(ComponentPose=Local To Component(LocalPose=Use cached pose 'presim'()), Alpha=1, ExternalForce="0,0,0") {
    angularDampingOverride=0.9999
    angularSpringConstant=25
    bAngularSpring=true
    bGravityOverrideInSimSpace=true
    bOverrideAngularDamping=true
    simulationSpace=BoneRelative
    alphaBoolBlend {blendOption=Linear}
    boundBone {boneName=dyn_jacket_bk_l_01}
    componentAppliedLinearAccClamp=<3500 3500 3500>
    componentLinearAccScale=<1.5 1.5 1.5>
    componentLinearVelScale=<1.5 1.5 1.5>
    physicsBodyDefinitions [{
      sphereCollisionRadius=0
      boundBone {boneName=dyn_jacket_bk_l_01}
      boxExtents=<10 20 20>
      constraintSetup {angularLimitsMax=<15 0 15>, angularLimitsMin=<-15 -30 -15>, angularTarget=<0 0 0>}
      localJointOffset=<2 0 0>
     }]
    relativeSpaceBone {boneName=dyn_simspace}
   }
   n5 = Anim Dynamics - Bone: dyn_jacket_bk_l_02(ComponentPose=@n4, Alpha=1, ExternalForce="0,0,0") {
    angularDampingOverride=0.9999
    angularSpringConstant=25
    bAngularSpring=true
    bGravityOverrideInSimSpace=true
    bOverrideAngularDamping=true
    simulationSpace=BoneRelative
    alphaBoolBlend {blendOption=Linear}
    boundBone {boneName=dyn_jacket_bk_l_02}
    componentAppliedLinearAccClamp=<3500 3500 3500>
    componentLinearAccScale=<1.5 1.5 1.5>
    componentLinearVelScale=<1.5 1.5 1.5>
    physicsBodyDefinitions [{
      sphereCollisionRadius=0
      boundBone {boneName=dyn_jacket_bk_l_02}
      boxExtents=<10 20 20>
      constraintSetup {angularLimitsMax=<15 0 15>, angularLimitsMin=<-15 -30 -15>, angularTarget=<0 0 0>}
      localJointOffset=<2 0 0>
     }]
    relativeSpaceBone {boneName=dyn_simspace}
   }
   n6 = Anim Dynamics - Bone: dyn_jacket_bk_l_03(ComponentPose=@n5, Alpha=1, ExternalForce="0,0,0") {
    angularDampingOverride=0.9999
    angularSpringConstant=25
    bAngularSpring=true
    bGravityOverrideInSimSpace=true
    bOverrideAngularDamping=true
    simulationSpace=BoneRelative
    alphaBoolBlend {blendOption=Linear}
    boundBone {boneName=dyn_jacket_bk_l_03}
    componentAppliedLinearAccClamp=<3500 3500 3500>
    componentLinearAccScale=<1.5 1.5 1.5>
    componentLinearVelScale=<1.5 1.5 1.5>
    physicsBodyDefinitions [{
      sphereCollisionRadius=0
      boundBone {boneName=dyn_jacket_bk_l_03}
      boxExtents=<10 20 20>
      constraintSetup {angularLimitsMax=<15 0 15>, angularLimitsMin=<-15 -30 -15>, angularTarget=<0 0 0>}
      localJointOffset=<2 0 0>
     }]
    relativeSpaceBone {boneName=dyn_simspace}
   }
   n7 = Anim Dynamics - Bone: dyn_jacket_bk_01(ComponentPose=@n6, Alpha=1, ExternalForce="0,0,0") {
    angularDampingOverride=0.9999
    angularSpringConstant=25
    bAngularSpring=true
    bGravityOverrideInSimSpace=true
    bOverrideAngularDamping=true
    simulationSpace=BoneRelative
    alphaBoolBlend {blendOption=Linear}
    boundBone {boneName=dyn_jacket_bk_01}
    componentAppliedLinearAccClamp=<3500 3500 3500>
    componentLinearAccScale=<1.5 1.5 1.5>
    componentLinearVelScale=<1.5 1.5 1.5>
    physicsBodyDefinitions [{
      sphereCollisionRadius=0
      boundBone {boneName=dyn_jacket_bk_01}
      boxExtents=<10 20 20>
      constraintSetup {angularLimitsMax=<15 0 15>, angularLimitsMin=<-15 -30 -15>, angularTarget=<0 0 0>}
      localJointOffset=<2 0 0>
     }]
    relativeSpaceBone {boneName=dyn_simspace}
   }
   n8 = Anim Dynamics - Bone: dyn_jacket_bk_02(ComponentPose=@n7, Alpha=1, ExternalForce="0,0,0") {
    angularDampingOverride=0.9999
    angularSpringConstant=25
    bAngularSpring=true
    bGravityOverrideInSimSpace=true
    bOverrideAngularDamping=true
    simulationSpace=BoneRelative
    alphaBoolBlend {blendOption=Linear}
    boundBone {boneName=dyn_jacket_bk_02}
    componentAppliedLinearAccClamp=<3500 3500 3500>
    componentLinearAccScale=<1.5 1.5 1.5>
    componentLinearVelScale=<1.5 1.5 1.5>
    physicsBodyDefinitions [{
      sphereCollisionRadius=0
      boundBone {boneName=dyn_jacket_bk_02}
      boxExtents=<10 20 20>
      constraintSetup {angularLimitsMax=<15 0 15>, angularLimitsMin=<-15 -30 -15>, angularTarget=<0 0 0>}
      localJointOffset=<2 0 0>
     }]
    relativeSpaceBone {boneName=dyn_simspace}
   }
   n9 = Anim Dynamics - Bone: dyn_jacket_bk_03(ComponentPose=@n8, Alpha=1, ExternalForce="0,0,0") {
    angularDampingOverride=0.9999
    angularSpringConstant=25
    bAngularSpring=true
    bGravityOverrideInSimSpace=true
    bOverrideAngularDamping=true
    simulationSpace=BoneRelative
    alphaBoolBlend {blendOption=Linear}
    boundBone {boneName=dyn_jacket_bk_03}
    componentAppliedLinearAccClamp=<3500 3500 3500>
    componentLinearAccScale=<1.5 1.5 1.5>
    componentLinearVelScale=<1.5 1.5 1.5>
    physicsBodyDefinitions [{
      sphereCollisionRadius=0
      boundBone {boneName=dyn_jacket_bk_03}
      boxExtents=<10 20 20>
      constraintSetup {angularLimitsMax=<15 0 15>, angularLimitsMin=<-15 -30 -15>, angularTarget=<0 0 0>}
      localJointOffset=<2 0 0>
     }]
    relativeSpaceBone {boneName=dyn_simspace}
   }
   n10 = Anim Dynamics - Bone: dyn_jacket_bk_r_01(ComponentPose=@n9, Alpha=1, ExternalForce="0,0,0") {
    angularDampingOverride=0.9999
    angularSpringConstant=25
    bAngularSpring=true
    bGravityOverrideInSimSpace=true
    bOverrideAngularDamping=true
    simulationSpace=BoneRelative
    alphaBoolBlend {blendOption=Linear}
    boundBone {boneName=dyn_jacket_bk_r_01}
    componentAppliedLinearAccClamp=<3500 3500 3500>
    componentLinearAccScale=<1.5 1.5 1.5>
    componentLinearVelScale=<1.5 1.5 1.5>
    physicsBodyDefinitions [{
      sphereCollisionRadius=0
      boundBone {boneName=dyn_jacket_bk_r_01}
      boxExtents=<10 20 20>
      constraintSetup {angularLimitsMax=<15 0 15>, angularLimitsMin=<-15 -30 -15>, angularTarget=<0 0 0>}
      localJointOffset=<2 0 0>
     }]
    relativeSpaceBone {boneName=dyn_simspace}
   }
   n11 = Anim Dynamics - Bone: dyn_jacket_bk_r_02(ComponentPose=@n10, Alpha=1, ExternalForce="0,0,0") {
    angularDampingOverride=0.9999
    angularSpringConstant=25
    bAngularSpring=true
    bGravityOverrideInSimSpace=true
    bOverrideAngularDamping=true
    simulationSpace=BoneRelative
    alphaBoolBlend {blendOption=Linear}
    boundBone {boneName=dyn_jacket_bk_r_02}
    componentAppliedLinearAccClamp=<3500 3500 3500>
    componentLinearAccScale=<1.5 1.5 1.5>
    componentLinearVelScale=<1.5 1.5 1.5>
    physicsBodyDefinitions [{
      sphereCollisionRadius=0
      boundBone {boneName=dyn_jacket_bk_r_02}
      boxExtents=<10 20 20>
      constraintSetup {angularLimitsMax=<15 0 15>, angularLimitsMin=<-15 -30 -15>, angularTarget=<0 0 0>}
      localJointOffset=<2 0 0>
     }]
    relativeSpaceBone {boneName=dyn_simspace}
   }
   n12 = Anim Dynamics - Bone: dyn_jacket_bk_r_03(ComponentPose=@n11, Alpha=1, ExternalForce="0,0,0") {
    angularDampingOverride=0.9999
    angularSpringConstant=25
    bAngularSpring=true
    bGravityOverrideInSimSpace=true
    bOverrideAngularDamping=true
    simulationSpace=BoneRelative
    alphaBoolBlend {blendOption=Linear}
    boundBone {boneName=dyn_jacket_bk_r_03}
    componentAppliedLinearAccClamp=<3500 3500 3500>
    componentLinearAccScale=<1.5 1.5 1.5>
    componentLinearVelScale=<1.5 1.5 1.5>
    physicsBodyDefinitions [{
      sphereCollisionRadius=0
      boundBone {boneName=dyn_jacket_bk_r_03}
      boxExtents=<10 20 20>
      constraintSetup {angularLimitsMax=<15 0 15>, angularLimitsMin=<-15 -30 -15>, angularTarget=<0 0 0>}
      localJointOffset=<2 0 0>
     }]
    relativeSpaceBone {boneName=dyn_simspace}
   }
   n13 = Component To Local(ComponentPose=@n12)
   n14 = Get bRigidBodyEnabled()
   n15 = RigidBody!(
    ComponentPose=Local To Component(LocalPose=Blend Poses by bool(BlendPose_0=@n3, BlendPose_1=@n13, BlendTime_0=0.5, BlendTime_1=0.5, bActiveValue=@n14))
    Alpha:float
    ExternalForce
   ) -> (Pose) {
    bClampLinearTranslationLimitToRefPose=true
    bOverrideWorldGravity=true
    overridePhysicsAsset=&/F_MED_Pliant_Collar_Physics
    simulationSpace=BaseBoneSpace
    alphaBoolBlend {blendOption=Linear}
    baseBoneRef {boneName=spine_05}
    componentAppliedLinearAccClamp=<3000 3000 3000>
    componentLinearAccScale=<0.45 0.25 0.25>
    componentLinearVelScale=<0.3 0.15 0.15>
    overrideWorldGravity {z=-300}
    simSpaceSettings {
     maxAngularAcceleration=1000
     maxAngularVelocity=3
     maxLinearAcceleration=700
     maxLinearVelocity=500
     velocityScaleZ=0
     worldAlpha=1
     externalLinearDragV=<1 1>
    }
   }
   Output Pose(Result=Component To Local(ComponentPose=@n15)) {layerGroup=None}
   n16 = "Transform (Modify) Bone - Bone: dyn_spine_04"!(
    ComponentPose=Local To Component(LocalPose=Use cached pose 'constraints'())
    bAlphaBoolEnabled:bool=Get bIsPlayingMeleeAnim()
    Translation="0.5,1,1"
    Rotation="0,0,0"
    Scale="1,1.4,1.4"
   ) -> (Pose) {
    alphaInputType=Bool
    rotationMode=BMM_Additive
    scaleMode=BMM_Additive
    translationMode=BMM_Additive
    alphaBoolBlend {blendInTime=0.1, blendOption=Sinusoidal, blendOutTime=0.1}
    boneToModify {boneName=dyn_spine_04}
    scale {y=1.4, z=1.4}
    translation=<0.5 1 1>
   }
   n17 = Get thigh_rot_l()
   n18 = "Transform (Modify) Bone - Bone: dyn_jacket_fr_l_MSTR"(ComponentPose=@n16, Alpha=@n17, Translation="-3.580,-5.888,-0.672", Rotation="6.273,-51.119,-109.721", Scale="1,1,1") {
    rotationMode=BMM_Additive
    rotationSpace=BCS_BoneSpace
    translationMode=BMM_Additive
    translationSpace=BCS_BoneSpace
    alphaBoolBlend {blendOption=Linear}
    alphaScaleBiasClamp {bInterpResult=true, interpSpeedDecreasing=5, interpSpeedIncreasing=40}
    boneToModify {boneName=dyn_jacket_fr_l_MSTR}
    rotation=<rot 6.273 -51.119 -109.721>
    translation=<-3.5799 -5.8879 -0.6719>
   }
   n19 = "Transform (Modify) Bone - Bone: dyn_jacket_l_MSTR"(ComponentPose=@n18, Alpha=@n17, Translation="-1.978,-4.139,2.886", Rotation="19.151,-14.809,-74.596", Scale="1,1,1") {
    rotationMode=BMM_Additive
    rotationSpace=BCS_BoneSpace
    translationMode=BMM_Additive
    translationSpace=BCS_BoneSpace
    alphaBoolBlend {blendOption=Linear}
    alphaScaleBiasClamp {bInterpResult=true, interpSpeedDecreasing=5, interpSpeedIncreasing=40}
    boneToModify {boneName=dyn_jacket_l_MSTR}
    rotation=<rot 19.151 -14.809 -74.596>
    translation=<-1.978 -4.139 2.886>
   }
   n20 = Get thigh_rot_r()
   n21 = "Transform (Modify) Bone - Bone: dyn_jacket_fr_r_MSTR"(ComponentPose=@n19, Alpha=@n20, Translation="-1.204,6.355,-1.972", Rotation="43.234,31.652,70.812", Scale="1,1,1") {
    rotationMode=BMM_Additive
    rotationSpace=BCS_BoneSpace
    translationMode=BMM_Additive
    translationSpace=BCS_BoneSpace
    alphaBoolBlend {blendOption=Linear}
    alphaScaleBiasClamp {bInterpResult=true, interpSpeedIncreasing=30}
    boneToModify {boneName=dyn_jacket_fr_r_MSTR}
    rotation=<rot 43.234 31.652 70.812>
    translation=<-1.204 6.355 -1.972>
   }
   n22 = "Transform (Modify) Bone - Bone: dyn_jacket_r_MSTR"(ComponentPose=@n21, Alpha=@n20, Translation="1.5,2,1", Rotation="40,12.739,8.143", Scale="1,1,1") {
    rotationMode=BMM_Additive
    rotationSpace=BCS_BoneSpace
    translationMode=BMM_Additive
    translationSpace=BCS_BoneSpace
    alphaBoolBlend {blendOption=Linear}
    alphaScaleBiasClamp {bInterpResult=true, interpSpeedDecreasing=5, interpSpeedIncreasing=30}
    boneToModify {boneName=dyn_jacket_r_MSTR}
    rotation=<rot 40 12.739 8.143>
    translation=<1.5 2 1>
   }
   n23 = Get spine back() -> (spine back:double)
   n24 = Look At - Bone: dyn_jacket_fr_l_MSTR!(ComponentPose=Local To Component(LocalPose=Use cached pose 'constraints'()), bAlphaBoolEnabled:bool=True) -> (Pose) {
    alphaInputType=Bool
    alphaBoolBlend {blendInTime=0.1, blendOption=Linear, blendOutTime=0.1}
    boneToModify {boneName=dyn_jacket_fr_l_MSTR}
    lookAtLocation=<0 2.5 -5>
    lookAtTarget {boneReference {boneName=calf_l}}
    lookAt_Axis {axis=<1 0>}
   }
   n25 = Look At - Bone: dyn_jacket_fr_r_MSTR!(ComponentPose=@n24, bAlphaBoolEnabled:bool=True) -> (Pose) {
    alphaInputType=Bool
    alphaBoolBlend {blendInTime=0.1, blendOption=Linear, blendOutTime=0.1}
    boneToModify {boneName=dyn_jacket_fr_r_MSTR}
    lookAtLocation=<0 -5 5>
    lookAtTarget {boneReference {boneName=calf_r}}
    lookAt_Axis {axis=<1 0>}
   }
   n26 = Get spine Left() -> (spine Left:double)
   n27 = "Transform (Modify) Bone - Bone: dyn_jacket_l_MSTR"(ComponentPose=@n25, Alpha=@n26, Translation="0,0,2", Rotation="10,0,0", Scale="1,1,1") {
    rotationMode=BMM_Additive
    rotationSpace=BCS_BoneSpace
    translationMode=BMM_Additive
    translationSpace=BCS_BoneSpace
    alphaBoolBlend {blendOption=Linear}
    alphaScaleBiasClamp {interpSpeedDecreasing=5, interpSpeedIncreasing=40}
    boneToModify {boneName=dyn_jacket_l_MSTR}
    rotation {pitch=10}
    translation {z=2}
   }
   n28 = "Transform (Modify) Bone - Bone: dyn_jacket_bk_l_MSTR"(ComponentPose=@n27, Alpha=@n26, Translation="0,0,2", Rotation="20,0,0", Scale="1,1,1") {
    rotationMode=BMM_Additive
    rotationSpace=BCS_BoneSpace
    translationMode=BMM_Additive
    translationSpace=BCS_BoneSpace
    alphaBoolBlend {blendOption=Linear}
    alphaScaleBiasClamp {interpSpeedDecreasing=5, interpSpeedIncreasing=40}
    boneToModify {boneName=dyn_jacket_bk_l_MSTR}
    rotation {pitch=20}
    translation {z=2}
   }
   n29 = "Transform (Modify) Bone - Bone: dyn_jacket_bk_MSTR"(ComponentPose=@n28, Alpha=@n26, Translation="0,0,0", Rotation="10,0,0", Scale="1,1,1") {
    rotationMode=BMM_Additive
    rotationSpace=BCS_BoneSpace
    translationMode=BMM_Additive
    translationSpace=BCS_BoneSpace
    alphaBoolBlend {blendOption=Linear}
    alphaScaleBiasClamp {interpSpeedDecreasing=5, interpSpeedIncreasing=40}
    boneToModify {boneName=dyn_jacket_bk_MSTR}
    rotation {pitch=10}
   }
   n30 = Get spine right() -> (spine right:double)
   n31 = "Transform (Modify) Bone - Bone: dyn_jacket_r_MSTR"(ComponentPose=@n29, Alpha=@n30, Translation="0,0,2", Rotation="10,0,0", Scale="1,1,1") {
    rotationMode=BMM_Additive
    rotationSpace=BCS_BoneSpace
    translationMode=BMM_Additive
    translationSpace=BCS_BoneSpace
    alphaBoolBlend {blendOption=Linear}
    alphaScaleBiasClamp {interpSpeedDecreasing=5, interpSpeedIncreasing=40}
    boneToModify {boneName=dyn_jacket_r_MSTR}
    rotation {pitch=10}
    translation {z=2}
   }
   n32 = "Transform (Modify) Bone - Bone: dyn_jacket_bk_r_MSTR"(ComponentPose=@n31, Alpha=@n30, Translation="0,0,2", Rotation="20,0,0", Scale="1,1,1") {
    rotationMode=BMM_Additive
    rotationSpace=BCS_BoneSpace
    translationMode=BMM_Additive
    translationSpace=BCS_BoneSpace
    alphaBoolBlend {blendOption=Linear}
    alphaScaleBiasClamp {interpSpeedDecreasing=5, interpSpeedIncreasing=40}
    boneToModify {boneName=dyn_jacket_bk_r_MSTR}
    rotation {pitch=20}
    translation {z=2}
   }
   n33 = "Transform (Modify) Bone - Bone: dyn_jacket_bk_MSTR"(ComponentPose=@n32, Alpha=@n30, Translation="0,0,0", Rotation="10,0,0", Scale="1,1,1") {
    rotationMode=BMM_Additive
    rotationSpace=BCS_BoneSpace
    translationMode=BMM_Additive
    translationSpace=BCS_BoneSpace
    alphaBoolBlend {blendOption=Linear}
    alphaScaleBiasClamp {interpSpeedDecreasing=5, interpSpeedIncreasing=40}
    boneToModify {boneName=dyn_jacket_bk_MSTR}
    rotation {pitch=10}
   }
   n34 = "Transform (Modify) Bone - Bone: dyn_jacket_bk_l_MSTR"(ComponentPose=@n33, Alpha=@n23, Translation="0,0,2", Rotation="10,0,0", Scale="1,1,1") {
    rotationMode=BMM_Additive
    rotationSpace=BCS_BoneSpace
    translationMode=BMM_Additive
    translationSpace=BCS_BoneSpace
    alphaBoolBlend {blendOption=Linear}
    alphaScaleBiasClamp {interpSpeedDecreasing=5, interpSpeedIncreasing=40}
    boneToModify {boneName=dyn_jacket_bk_l_MSTR}
    rotation {pitch=10}
    translation {z=2}
   }
   n35 = "Transform (Modify) Bone - Bone: dyn_jacket_bk_r_MSTR"(ComponentPose=@n34, Alpha=@n23, Translation="0,0,2", Rotation="10,0,0", Scale="1,1,1") {
    rotationMode=BMM_Additive
    rotationSpace=BCS_BoneSpace
    translationMode=BMM_Additive
    translationSpace=BCS_BoneSpace
    alphaBoolBlend {blendOption=Linear}
    alphaScaleBiasClamp {interpSpeedDecreasing=5, interpSpeedIncreasing=40}
    boneToModify {boneName=dyn_jacket_bk_r_MSTR}
    rotation {pitch=10}
    translation {z=2}
   }
   n36 = "Transform (Modify) Bone - Bone: dyn_jacket_bk_MSTR"(ComponentPose=@n35, Alpha=@n23, Translation="0,0,0", Rotation="5,0,0", Scale="1,1,1") {
    rotationMode=BMM_Additive
    rotationSpace=BCS_BoneSpace
    translationMode=BMM_Additive
    translationSpace=BCS_BoneSpace
    alphaBoolBlend {blendOption=Linear}
    alphaScaleBiasClamp {interpSpeedDecreasing=5, interpSpeedIncreasing=40}
    boneToModify {boneName=dyn_jacket_bk_MSTR}
    rotation {pitch=5}
   }
   n37 = "Transform (Modify) Bone - Bone: dyn_jacket_fr_r_MSTR"(ComponentPose=@n36, Alpha=1, Translation="0,0,0", Rotation="5,0,0", Scale="1,1,1") {
    rotationMode=BMM_Additive
    rotationSpace=BCS_BoneSpace
    translationMode=BMM_Additive
    translationSpace=BCS_BoneSpace
    alphaBoolBlend {blendOption=Linear}
    alphaScaleBiasClamp {interpSpeedDecreasing=5, interpSpeedIncreasing=40}
    boneToModify {boneName=dyn_jacket_fr_r_MSTR}
    rotation {pitch=5}
   }
   n38 = NOT Boolean(A=Get bRigidBodyEnabled())
   n39 = "Transform (Modify) Bone - Bone: dyn_jacket_fr_l_MSTR"!(ComponentPose=@n37, bAlphaBoolEnabled:bool=@n38, Translation="0,0,0", Rotation="6.191,6.081,-2.357", Scale="1,1,1") -> (Pose) {
    alphaInputType=Bool
    rotationMode=BMM_Additive
    rotationSpace=BCS_BoneSpace
    translationMode=BMM_Additive
    translationSpace=BCS_BoneSpace
    alphaBoolBlend {blendOption=Linear}
    alphaScaleBiasClamp {interpSpeedDecreasing=5, interpSpeedIncreasing=40}
    boneToModify {boneName=dyn_jacket_fr_l_MSTR}
    rotation=<rot 6.191 6.081 -2.357>
   }
   n40 = "Transform (Modify) Bone - Bone: dyn_jacket_fr_r_MSTR"!(ComponentPose=@n39, bAlphaBoolEnabled:bool=@n38, Translation="0,0,0", Rotation="0.012,-3.204,0.000", Scale="1,1,1") -> (Pose) {
    alphaInputType=Bool
    rotationMode=BMM_Additive
    rotationSpace=BCS_BoneSpace
    translationMode=BMM_Additive
    translationSpace=BCS_BoneSpace
    alphaBoolBlend {blendOption=Linear}
    alphaScaleBiasClamp {interpSpeedDecreasing=5, interpSpeedIncreasing=40}
    boneToModify {boneName=dyn_jacket_fr_r_MSTR}
    rotation=<rot 0.012 -3.2041 0>
   }
   Save cached pose 'presim'(Pose=Blend Poses by bool(BlendPose_0=Component To Local(ComponentPose=@n22), BlendPose_1=Component To Local(ComponentPose=@n40), bActiveValue=Get bIsCrouching()) {blendType=Linear}) {cacheName=presim}
   n41 = Control Rig!(Source=Copy Pose From Mesh() {bUseAttachedParent=true}, Alpha, rifle:bool=Get Rifle() -> (Rifle:bool)) -> (Pose) {
    alphaBoolBlend {blendOption=Linear}
    controlRigAssetReference {blueprintRigClass=&/F_MED_Pliant_DeformRig_Master_ControlRig.F_MED_Pliant_DeformRig_Master_ControlRig_C}
   }
   n42 = "Constraint - dyn_jacket_fr_l_Look_at (PPPPPPPP)"!(
    ComponentPose=Local To Component(LocalPose=@n41)
    Alpha=1
    ConstraintWeights_0=0
    ConstraintWeights_1:float=0
    ConstraintWeights_2:float=0.0178
    ConstraintWeights_3:float=0.048
    ConstraintWeights_4:float=0.1301
    ConstraintWeights_5:float=0.8035
    ConstraintWeights_6:float=0.0007
    ConstraintWeights_7:float=0
   ) -> (Pose) {
    alphaBoolBlend {blendOption=Linear}
    boneToModify {boneName=dyn_jacket_fr_l_Look_at}
    constraintSetup [
     {targetBone {boneName=clavicle_l}}
     {targetBone {boneName=upperarm_twist_01_r}}
     {targetBone {boneName=spine_01}}
     {targetBone {boneName=spine_02}}
     {targetBone {boneName=spine_03}}
     {targetBone {boneName=spine_04}}
     {targetBone {boneName=spine_05}}
     {targetBone {boneName=clavicle_r}}
    ]
    constraintWeights [1,1,1,1,1,1,1,1]
   }
   n43 = "Constraint - dyn_jacket_fr_r_Look_at (PPPPPPPP)"!(
    ComponentPose=@n42
    Alpha=1
    ConstraintWeights_0=0
    ConstraintWeights_1:float=0
    ConstraintWeights_2:float=0.0178
    ConstraintWeights_3:float=0.048
    ConstraintWeights_4:float=0.1301
    ConstraintWeights_5:float=0.8035
    ConstraintWeights_6:float=0.0007
    ConstraintWeights_7:float=0
   ) -> (Pose) {
    alphaBoolBlend {blendOption=Linear}
    boneToModify {boneName=dyn_jacket_fr_r_Look_at}
    constraintSetup [
     {targetBone {boneName=clavicle_l}}
     {targetBone {boneName=upperarm_twist_01_r}}
     {targetBone {boneName=spine_01}}
     {targetBone {boneName=spine_02}}
     {targetBone {boneName=spine_03}}
     {targetBone {boneName=spine_04}}
     {targetBone {boneName=spine_05}}
     {targetBone {boneName=clavicle_r}}
    ]
    constraintWeights [1,1,1,1,1,1,1,1]
   }
   n44 = "Constraint - dyn_jacket_r_Look_at (PPPPPPPP)"!(
    ComponentPose=@n43
    Alpha=1
    ConstraintWeights_0=0
    ConstraintWeights_1:float=0.0001
    ConstraintWeights_2:float=0.0241
    ConstraintWeights_3:float=0.0698
    ConstraintWeights_4:float=0.1472
    ConstraintWeights_5:float=0.6995
    ConstraintWeights_6:float=0.0397
    ConstraintWeights_7:float=0.0197
   ) -> (Pose) {
    alphaBoolBlend {blendOption=Linear}
    boneToModify {boneName=dyn_jacket_r_Look_at}
    constraintSetup [
     {targetBone {boneName=clavicle_l}}
     {targetBone {boneName=upperarm_twist_01_r}}
     {targetBone {boneName=spine_01}}
     {targetBone {boneName=spine_02}}
     {targetBone {boneName=spine_03}}
     {targetBone {boneName=spine_04}}
     {targetBone {boneName=spine_05}}
     {targetBone {boneName=clavicle_r}}
    ]
    constraintWeights [1,1,1,1,1,1,1,1]
   }
   n45 = "Constraint - dyn_jacket_l_Look_at (PPPPPPPP)"!(
    ComponentPose=@n44
    Alpha=1
    ConstraintWeights_0=0.0299
    ConstraintWeights_1:float=0
    ConstraintWeights_2:float=0.0225
    ConstraintWeights_3:float=0.0647
    ConstraintWeights_4:float=0.1373
    ConstraintWeights_5:float=0.7047
    ConstraintWeights_6:float=0.0409
    ConstraintWeights_7:float=0
   ) -> (Pose) {
    alphaBoolBlend {blendOption=Linear}
    boneToModify {boneName=dyn_jacket_l_Look_at}
    constraintSetup [
     {targetBone {boneName=clavicle_l}}
     {targetBone {boneName=upperarm_twist_01_r}}
     {targetBone {boneName=spine_01}}
     {targetBone {boneName=spine_02}}
     {targetBone {boneName=spine_03}}
     {targetBone {boneName=spine_04}}
     {targetBone {boneName=spine_05}}
     {targetBone {boneName=clavicle_r}}
    ]
    constraintWeights [1,1,1,1,1,1,1,1]
   }
   n46 = "Constraint - dyn_jacket_bk_Look_at (PPPPPPPP)"!(
    ComponentPose=@n45
    Alpha=1
    ConstraintWeights_0=0.0256
    ConstraintWeights_1:float=0.0011
    ConstraintWeights_2:float=0.07
    ConstraintWeights_3:float=0.1786
    ConstraintWeights_4:float=0.2038
    ConstraintWeights_5:float=0.4146
    ConstraintWeights_6:float=0.0935
    ConstraintWeights_7:float=0.0128
   ) -> (Pose) {
    alphaBoolBlend {blendOption=Linear}
    boneToModify {boneName=dyn_jacket_bk_Look_at}
    constraintSetup [
     {targetBone {boneName=clavicle_l}}
     {targetBone {boneName=upperarm_twist_01_r}}
     {targetBone {boneName=spine_01}}
     {targetBone {boneName=spine_02}}
     {targetBone {boneName=spine_03}}
     {targetBone {boneName=spine_04}}
     {targetBone {boneName=spine_05}}
     {targetBone {boneName=clavicle_r}}
    ]
    constraintWeights [1,1,1,1,1,1,1,1]
   }
   n47 = "Constraint - dyn_jacket_bk_r_Look_at (PPPPPPPP)"!(
    ComponentPose=@n46
    Alpha=1
    ConstraintWeights_0=0.0022
    ConstraintWeights_1:float=0.0045
    ConstraintWeights_2:float=0.05
    ConstraintWeights_3:float=0.1418
    ConstraintWeights_4:float=0.1886
    ConstraintWeights_5:float=0.4922
    ConstraintWeights_6:float=0.0746
    ConstraintWeights_7:float=0.0461
   ) -> (Pose) {
    alphaBoolBlend {blendOption=Linear}
    boneToModify {boneName=dyn_jacket_bk_r_Look_at}
    constraintSetup [
     {targetBone {boneName=clavicle_l}}
     {targetBone {boneName=upperarm_twist_01_r}}
     {targetBone {boneName=spine_01}}
     {targetBone {boneName=spine_02}}
     {targetBone {boneName=spine_03}}
     {targetBone {boneName=spine_04}}
     {targetBone {boneName=spine_05}}
     {targetBone {boneName=clavicle_r}}
    ]
    constraintWeights [1,1,1,1,1,1,1,1]
   }
   n48 = "Constraint - dyn_jacket_bk_l_Look_at (PPPPPPPP)"!(
    ComponentPose=@n47
    Alpha=1
    ConstraintWeights_0=0.0509
    ConstraintWeights_1:float=0.0001
    ConstraintWeights_2:float=0.0512
    ConstraintWeights_3:float=0.143
    ConstraintWeights_4:float=0.1938
    ConstraintWeights_5:float=0.4832
    ConstraintWeights_6:float=0.0771
    ConstraintWeights_7:float=0.0007
   ) -> (Pose) {
    alphaBoolBlend {blendOption=Linear}
    boneToModify {boneName=dyn_jacket_bk_l_Look_at}
    constraintSetup [
     {targetBone {boneName=clavicle_l}}
     {targetBone {boneName=upperarm_twist_01_r}}
     {targetBone {boneName=spine_01}}
     {targetBone {boneName=spine_02}}
     {targetBone {boneName=spine_03}}
     {targetBone {boneName=spine_04}}
     {targetBone {boneName=spine_05}}
     {targetBone {boneName=clavicle_r}}
    ]
    constraintWeights [1,1,1,1,1,1,1,1]
   }
   n49 = Look At - Bone: dyn_jacket_bk_r_Look_at(ComponentPose=@n48, Alpha=1) {
    alphaBoolBlend {blendOption=Linear}
    boneToModify {boneName=dyn_jacket_bk_r_Look_at}
    lookAtLocation {x=0}
    lookAtTarget {boneReference {boneName=pelvis}}
    lookAt_Axis {axis=<1 0>}
   }
   n50 = Look At - Bone: dyn_jacket_bk_Look_at(ComponentPose=@n49, Alpha=1) {
    alphaBoolBlend {blendOption=Linear}
    boneToModify {boneName=dyn_jacket_bk_Look_at}
    lookAtLocation {x=0}
    lookAtTarget {boneReference {boneName=pelvis}}
    lookAt_Axis {axis=<1 0>}
   }
   n51 = Look At - Bone: dyn_jacket_bk_l_Look_at(ComponentPose=@n50, Alpha=1) {
    alphaBoolBlend {blendOption=Linear}
    boneToModify {boneName=dyn_jacket_bk_l_Look_at}
    lookAtLocation {x=0}
    lookAtTarget {boneReference {boneName=pelvis}}
    lookAt_Axis {axis=<1 0>}
   }
   n52 = Look At - Bone: dyn_jacket_l_Look_at(ComponentPose=@n51, Alpha=1) {
    alphaBoolBlend {blendOption=Linear}
    boneToModify {boneName=dyn_jacket_l_Look_at}
    lookAtLocation {x=0}
    lookAtTarget {boneReference {boneName=pelvis}}
    lookAt_Axis {axis=<1 0>}
   }
   n53 = Look At - Bone: dyn_jacket_fr_l_Look_at(ComponentPose=@n52, Alpha=1) {
    alphaBoolBlend {blendOption=Linear}
    boneToModify {boneName=dyn_jacket_fr_l_Look_at}
    lookAtLocation {x=0}
    lookAtTarget {boneReference {boneName=dyn_jacket_tgt}}
    lookAt_Axis {axis=<1 0>}
   }
   n54 = Look At - Bone: dyn_jacket_fr_r_Look_at(ComponentPose=@n53, Alpha=1) {
    alphaBoolBlend {blendOption=Linear}
    boneToModify {boneName=dyn_jacket_fr_r_Look_at}
    lookAtLocation {x=0}
    lookAtTarget {boneReference {boneName=dyn_jacket_tgt}}
    lookAt_Axis {axis=<1 0>}
   }
   n55 = Look At - Bone: dyn_jacket_r_Look_at(ComponentPose=@n54, Alpha=1) {
    alphaBoolBlend {blendOption=Linear}
    boneToModify {boneName=dyn_jacket_r_Look_at}
    lookAtLocation {x=0}
    lookAtTarget {boneReference {boneName=pelvis}}
    lookAt_Axis {axis=<1 0>}
   }
   n56 = "Constraint - dyn_Belt_ring_MSTR (PPPP)"!(
    ComponentPose=@n55
    Alpha=1
    ConstraintWeights_0=0.6614
    ConstraintWeights_1:float=0.3143
    ConstraintWeights_2:float=0.0006
    ConstraintWeights_3:float=0.0237
   ) -> (Pose) {
    alphaBoolBlend {blendOption=Linear}
    boneToModify {boneName=dyn_Belt_ring_MSTR}
    constraintSetup [
     {targetBone {boneName=pelvis}}
     {targetBone {boneName=thigh_twist_01_l}}
     {targetBone {boneName=thigh_twist_01_r}}
     {targetBone {boneName=thigh_l}}
    ]
    constraintWeights [1,1,1,1]
   }
   n57 = "Constraint - dyn_Belt_ring (RR)"!(ComponentPose=@n56, Alpha=1, ConstraintWeights_0=0, ConstraintWeights_1:float=1) -> (Pose) {
    alphaBoolBlend {blendOption=Linear}
    boneToModify {boneName=dyn_Belt_ring}
    constraintSetup [{transformType=Rotation, targetBone {boneName=thigh_l}}, {transformType=Rotation, targetBone {boneName=thigh_twist_01_l}}]
    constraintWeights [1,1]
   }
   n58 = "Transform (Modify) Bone - Bone: dyn_Belt_ring"(ComponentPose=@n57, Alpha=Get thigh_rot_l(), Translation="0,0,0", Rotation="-25,-10,-50", Scale="1,1,1") {rotationMode=BMM_Additive, alphaBoolBlend {blendOption=Linear}, boneToModify {boneName=dyn_Belt_ring}, rotation=<rot -25 -10 -50>}
   n59 = "Constraint - dyn_belt_ring_r_MSTR (PPP)"!(ComponentPose=@n58, Alpha=1, ConstraintWeights_0=0.2312, ConstraintWeights_1:float=0.4434, ConstraintWeights_2:float=0.3254) -> (Pose) {
    alphaBoolBlend {blendOption=Linear}
    boneToModify {boneName=dyn_belt_ring_r_MSTR}
    constraintSetup [{targetBone {boneName=thigh_r}}, {targetBone {boneName=pelvis}}, {targetBone {boneName=thigh_twist_01_r}}]
    constraintWeights [1,1,1]
   }
   n60 = "Constraint - dyn_belt_ring_r (RR)"!(ComponentPose=@n59, Alpha=1, ConstraintWeights_0=0, ConstraintWeights_1:float=1) -> (Pose) {
    alphaBoolBlend {blendOption=Linear}
    boneToModify {boneName=dyn_belt_ring_r}
    constraintSetup [{transformType=Rotation, targetBone {boneName=thigh_r}}, {transformType=Rotation, targetBone {boneName=thigh_twist_01_r}}]
    constraintWeights [1,1]
   }
   n61 = "Constraint - dyn_belt_ring_bk_r (PPPPPPPP)"!(
    ComponentPose=@n60
    Alpha=1
    ConstraintWeights_0=0.1054
    ConstraintWeights_1:float=0.0509
    ConstraintWeights_2:float=0.1431
    ConstraintWeights_3:float=0.0004
    ConstraintWeights_4:float=0.008
    ConstraintWeights_5:float=0.0008
    ConstraintWeights_6:float=0.6873
    ConstraintWeights_7:float=0.0041
   ) -> (Pose) {
    alphaBoolBlend {blendOption=Linear}
    boneToModify {boneName=dyn_belt_ring_bk_r}
    constraintSetup [
     {targetBone {boneName=dyn_glute_r}}
     {targetBone {boneName=thigh_r}}
     {targetBone {boneName=thigh_twist_01_r}}
     {targetBone {boneName=spine_01}}
     {targetBone {boneName=thigh_twist_01_l}}
     {targetBone {boneName=thigh_l}}
     {targetBone {boneName=pelvis}}
     {targetBone {boneName=dyn_glute_l}}
    ]
    constraintWeights [1,1,1,1,1,1,1,1]
   }
   n62 = "Constraint - dyn_belt_ring_bk_l (PPPPPPPP)"!(
    ComponentPose=@n61
    Alpha=1
    ConstraintWeights_0=0.091
    ConstraintWeights_1:float=0.0067
    ConstraintWeights_2:float=0.0019
    ConstraintWeights_3:float=0.0016
    ConstraintWeights_4:float=0.2378
    ConstraintWeights_5:float=0.0665
    ConstraintWeights_6:float=0.525
    ConstraintWeights_7:float=0.0696
   ) -> (Pose) {
    alphaBoolBlend {blendOption=Linear}
    boneToModify {boneName=dyn_belt_ring_bk_l}
    constraintSetup [
     {targetBone {boneName=dyn_glute_r}}
     {targetBone {boneName=thigh_twist_01_r}}
     {targetBone {boneName=spine_01}}
     {targetBone {boneName=spine_02}}
     {targetBone {boneName=thigh_twist_01_l}}
     {targetBone {boneName=thigh_l}}
     {targetBone {boneName=pelvis}}
     {targetBone {boneName=dyn_glute_l}}
    ]
    constraintWeights [1,1,1,1,1,1,1,1]
   }
   n63 = "Transform (Modify) Bone - Bone: dyn_belt_loop_MSTR"(ComponentPose=@n62, Alpha=Get thigh_rot_l(), Translation="0,0,0", Rotation="5.061,-12.871,25", Scale="1,1,1") {
    rotationMode=BMM_Additive
    translationMode=BMM_Additive
    alphaBoolBlend {blendOption=Linear}
    boneToModify {boneName=dyn_belt_loop_MSTR}
    rotation=<rot 5.061 -12.871 25>
   }
   n64 = "Transform (Modify) Bone - Bone: dyn_belt_r_MSTR"(ComponentPose=@n63, Alpha=Get thigh_rot_r(), Translation="0.000,-3,0.000", Rotation="0.000,0.000,20", Scale="1,1,1") {
    rotationMode=BMM_Additive
    translationMode=BMM_Additive
    alphaBoolBlend {blendOption=Linear}
    boneToModify {boneName=dyn_belt_r_MSTR}
    rotation {roll=20}
    translation {y=-3}
   }
   n65 = "Transform (Modify) Bone - Bone: dyn_Belt_ring_belt_MSTR"!(ComponentPose=@n64, bAlphaBoolEnabled:bool=True, Translation="0,0,0", Rotation="-16.690,-30.620,55.896", Scale="1,1,1") -> (Pose) {
    alphaInputType=Bool
    rotationMode=BMM_Additive
    translationMode=BMM_Additive
    alphaBoolBlend {blendOption=Linear}
    boneToModify {boneName=dyn_Belt_ring_belt_MSTR}
    rotation=<rot -16.69 -30.62 55.896>
   }
   n66 = NOT Boolean(A=Get bIsSkyDiving())
   n67 = Anim Dynamics - Bone: dyn_belt_loop_01!(ComponentPose=@n65, bAlphaBoolEnabled=@n66, GravityOverride:Vector="0,0,0", ExternalForce="0,0,0") -> (Pose) {
    alphaInputType=Bool
    angularDampingOverride=0.999
    angularSpringConstant=150
    bAngularSpring=true
    bGravityOverrideInSimSpace=true
    bOverrideAngularDamping=true
    simulationSpace=RootRelative
    alphaBoolBlend {blendInTime=0.1, blendOption=Linear, blendOutTime=0.1}
    boundBone {boneName=dyn_belt_loop_01}
    componentAppliedLinearAccClamp=<3500 3500 4300>
    componentLinearAccScale=<3 3 3>
    componentLinearVelScale=<3 3 3>
    physicsBodyDefinitions [{
      sphereCollisionRadius=0
      boundBone {boneName=dyn_belt_loop_01}
      boxExtents=<9 18 18>
      constraintSetup {angularLimitsMax=<0 0 40>, angularLimitsMin=<-30 -30 -30>, angularTarget=<0 0 0>}
      localJointOffset=<2 0 0>
     }]
    relativeSpaceBone {boneName=root}
   }
   n68 = Anim Dynamics - Bone: dyn_belt_loop_02!(ComponentPose=@n67, bAlphaBoolEnabled:bool=@n66, GravityOverride:Vector="0,0,0", ExternalForce="0,0,0") -> (Pose) {
    alphaInputType=Bool
    angularDampingOverride=0.999
    angularSpringConstant=150
    bAngularSpring=true
    bGravityOverrideInSimSpace=true
    bOverrideAngularDamping=true
    simulationSpace=RootRelative
    alphaBoolBlend {blendInTime=0.1, blendOption=Linear, blendOutTime=0.1}
    boundBone {boneName=dyn_belt_loop_02}
    componentAppliedLinearAccClamp=<3500 3500 4300>
    componentLinearAccScale=<3 3 3>
    componentLinearVelScale=<3 3 3>
    physicsBodyDefinitions [{
      sphereCollisionRadius=0
      boundBone {boneName=dyn_belt_loop_02}
      boxExtents=<9 18 18>
      constraintSetup {angularLimitsMax=<0 5 40>, angularLimitsMin=<-30 -30 -30>, angularTarget=<0 0 0>}
      localJointOffset=<2 0 0>
     }]
    relativeSpaceBone {boneName=root}
   }
   n69 = Anim Dynamics - Bone: dyn_belt_loop_03!(ComponentPose=@n68, bAlphaBoolEnabled:bool=@n66, GravityOverride:Vector="0,0,0", ExternalForce="0,0,0") -> (Pose) {
    alphaInputType=Bool
    angularDampingOverride=0.999
    angularSpringConstant=150
    bAngularSpring=true
    bGravityOverrideInSimSpace=true
    bOverrideAngularDamping=true
    simulationSpace=RootRelative
    alphaBoolBlend {blendInTime=0.1, blendOption=Linear, blendOutTime=0.1}
    boundBone {boneName=dyn_belt_loop_03}
    componentAppliedLinearAccClamp=<3500 3500 4300>
    componentLinearAccScale=<3 3 3>
    componentLinearVelScale=<3 3 3>
    physicsBodyDefinitions [{
      sphereCollisionRadius=0
      boundBone {boneName=dyn_belt_loop_03}
      boxExtents=<9 18 18>
      constraintSetup {angularLimitsMax=<0 10 40>, angularLimitsMin=<-30 -30 -20>, angularTarget=<0 0 0>}
      localJointOffset=<2 0 0>
     }]
    relativeSpaceBone {boneName=root}
   }
   n70 = Get GravityOverride()
   n71 = Get bIsSkyDiving()
   n72 = Anim Dynamics - Bone: dyn_belt_loop_01(ComponentPose=@n69, bAlphaBoolEnabled=@n71, GravityOverride=@n70, ExternalForce="0,0,0") {
    alphaInputType=Bool
    angularDampingOverride=0.999
    angularSpringConstant=150
    bAngularSpring=true
    bGravityOverrideInSimSpace=true
    bOverrideAngularDamping=true
    bUseGravityOverride=true
    simulationSpace=RootRelative
    alphaBoolBlend {blendInTime=0.1, blendOption=Linear, blendOutTime=0.1}
    boundBone {boneName=dyn_belt_loop_01}
    componentAppliedLinearAccClamp=<3500 3500 4300>
    componentLinearAccScale=<3 3 3>
    componentLinearVelScale=<3 3 3>
    physicsBodyDefinitions [{
      sphereCollisionRadius=0
      boundBone {boneName=dyn_belt_loop_01}
      boxExtents=<9 18 18>
      constraintSetup {angularLimitsMax=<0 0 40>, angularLimitsMin=<-30 -30 -30>, angularTarget=<0 0 0>}
      localJointOffset=<2 0 0>
     }]
    relativeSpaceBone {boneName=root}
   }
   n73 = Anim Dynamics - Bone: dyn_belt_loop_02(ComponentPose=@n72, bAlphaBoolEnabled=@n71, GravityOverride=@n70, ExternalForce="0,0,0") {
    alphaInputType=Bool
    angularDampingOverride=0.999
    angularSpringConstant=150
    bAngularSpring=true
    bGravityOverrideInSimSpace=true
    bOverrideAngularDamping=true
    bUseGravityOverride=true
    simulationSpace=RootRelative
    alphaBoolBlend {blendInTime=0.1, blendOption=Linear, blendOutTime=0.1}
    boundBone {boneName=dyn_belt_loop_02}
    componentAppliedLinearAccClamp=<3500 3500 4300>
    componentLinearAccScale=<3 3 3>
    componentLinearVelScale=<3 3 3>
    physicsBodyDefinitions [{
      sphereCollisionRadius=0
      boundBone {boneName=dyn_belt_loop_02}
      boxExtents=<9 18 18>
      constraintSetup {angularLimitsMax=<0 5 40>, angularLimitsMin=<-30 -30 -30>, angularTarget=<0 0 0>}
      localJointOffset=<2 0 0>
     }]
    relativeSpaceBone {boneName=root}
   }
   n74 = Anim Dynamics - Bone: dyn_belt_loop_03(ComponentPose=@n73, bAlphaBoolEnabled=@n71, GravityOverride=@n70, ExternalForce="0,0,0") {
    alphaInputType=Bool
    angularDampingOverride=0.999
    angularSpringConstant=150
    bAngularSpring=true
    bGravityOverrideInSimSpace=true
    bOverrideAngularDamping=true
    bUseGravityOverride=true
    simulationSpace=RootRelative
    alphaBoolBlend {blendInTime=0.1, blendOption=Linear, blendOutTime=0.1}
    boundBone {boneName=dyn_belt_loop_03}
    componentAppliedLinearAccClamp=<3500 3500 4300>
    componentLinearAccScale=<3 3 3>
    componentLinearVelScale=<3 3 3>
    physicsBodyDefinitions [{
      sphereCollisionRadius=0
      boundBone {boneName=dyn_belt_loop_03}
      boxExtents=<9 18 18>
      constraintSetup {angularLimitsMax=<0 10 40>, angularLimitsMin=<-30 -30 -20>, angularTarget=<0 0 0>}
      localJointOffset=<2 0 0>
     }]
    relativeSpaceBone {boneName=root}
   }
   n75 = Anim Dynamics - Bone: dyn_belt_r_01(ComponentPose=@n74, Alpha=1, ExternalForce="0,0,0") {
    angularDampingOverride=0.9999
    angularSpringConstant=150
    bAngularSpring=true
    bGravityOverrideInSimSpace=true
    bOverrideAngularDamping=true
    simulationSpace=RootRelative
    alphaBoolBlend {blendOption=Linear}
    boundBone {boneName=dyn_belt_r_01}
    componentAppliedLinearAccClamp=<3500 3500 3500>
    componentLinearAccScale=<3 3 3>
    componentLinearVelScale=<3 3 3>
    physicsBodyDefinitions [{
      sphereCollisionRadius=0
      boundBone {boneName=dyn_belt_r_01}
      boxExtents=<10 20 20>
      constraintSetup {angularLimitsMax=<10 0 15>, angularLimitsMin=<-30 -30 -15>, angularTarget=<0 0 0>}
      localJointOffset=<2 0 -0.25>
     }]
   }
   n76 = Anim Dynamics - Bone: dyn_belt_r_02(ComponentPose=@n75, Alpha=1, ExternalForce="0,0,0") {
    angularDampingOverride=0.9999
    angularSpringConstant=150
    bAngularSpring=true
    bGravityOverrideInSimSpace=true
    bOverrideAngularDamping=true
    simulationSpace=RootRelative
    alphaBoolBlend {blendOption=Linear}
    boundBone {boneName=dyn_belt_r_02}
    componentAppliedLinearAccClamp=<3500 3500 3500>
    componentLinearAccScale=<3 3 3>
    componentLinearVelScale=<3 3 3>
    physicsBodyDefinitions [{
      sphereCollisionRadius=0
      boundBone {boneName=dyn_belt_r_02}
      boxExtents=<10 20 20>
      constraintSetup {angularLimitsMax=<10 0 15>, angularLimitsMin=<-30 -30 -15>, angularTarget=<0 0 0>}
      localJointOffset=<2 0 -0.25>
     }]
   }
   n77 = Anim Dynamics - Bone: dyn_belt_r_03(ComponentPose=@n76, Alpha=1, ExternalForce="0,0,0") {
    angularDampingOverride=0.9999
    angularSpringConstant=150
    bAngularSpring=true
    bGravityOverrideInSimSpace=true
    bOverrideAngularDamping=true
    simulationSpace=RootRelative
    alphaBoolBlend {blendOption=Linear}
    boundBone {boneName=dyn_belt_r_03}
    componentAppliedLinearAccClamp=<3500 3500 3500>
    componentLinearAccScale=<3 3 3>
    componentLinearVelScale=<3 3 3>
    physicsBodyDefinitions [{
      sphereCollisionRadius=0
      boundBone {boneName=dyn_belt_r_03}
      boxExtents=<10 20 20>
      constraintSetup {angularLimitsMax=<10 0 15>, angularLimitsMin=<-30 -30 -15>, angularTarget=<0 0 0>}
      localJointOffset=<2 0 -0.25>
     }]
   }
   n78 = Get Move right() -> (Move right:double)
   n79 = "Transform (Modify) Bone - Bone: dyn_jacket_r_MSTR"(ComponentPose=@n77, Alpha=@n78, Translation="0,0,0", Rotation="5,0,0", Scale="1,1,1") {
    rotationMode=BMM_Additive
    rotationSpace=BCS_BoneSpace
    translationMode=BMM_Additive
    translationSpace=BCS_BoneSpace
    alphaBoolBlend {blendOption=Linear}
    alphaScaleBiasClamp {bInterpResult=true}
    boneToModify {boneName=dyn_jacket_r_MSTR}
    rotation {pitch=5}
   }
   n80 = "Transform (Modify) Bone - Bone: dyn_jacket_bk_r_MSTR"(ComponentPose=@n79, Alpha=@n78, Translation="0,0,0", Rotation="5,0,0", Scale="1,1,1") {
    rotationMode=BMM_Additive
    rotationSpace=BCS_BoneSpace
    translationMode=BMM_Additive
    translationSpace=BCS_BoneSpace
    alphaBoolBlend {blendOption=Linear}
    alphaScaleBiasClamp {bInterpResult=true}
    boneToModify {boneName=dyn_jacket_bk_r_MSTR}
    rotation {pitch=5}
   }
   n81 = Get Move lft() -> (Move lft:double)
   n82 = "Transform (Modify) Bone - Bone: dyn_jacket_l_MSTR"(ComponentPose=@n80, Alpha=@n81, Translation="0,0,0", Rotation="5,0,0", Scale="1,1,1") {
    rotationMode=BMM_Additive
    rotationSpace=BCS_BoneSpace
    translationMode=BMM_Additive
    translationSpace=BCS_BoneSpace
    alphaBoolBlend {blendOption=Linear}
    alphaScaleBiasClamp {bInterpResult=true}
    boneToModify {boneName=dyn_jacket_l_MSTR}
    rotation {pitch=5}
   }
   n83 = "Transform (Modify) Bone - Bone: dyn_jacket_bk_l_MSTR"(ComponentPose=@n82, Alpha=@n81, Translation="0,0,0", Rotation="5,0,0", Scale="1,1,1") {
    rotationMode=BMM_Additive
    rotationSpace=BCS_BoneSpace
    translationMode=BMM_Additive
    translationSpace=BCS_BoneSpace
    alphaBoolBlend {blendOption=Linear}
    alphaScaleBiasClamp {bInterpResult=true}
    boneToModify {boneName=dyn_jacket_bk_l_MSTR}
    rotation {pitch=5}
   }
   n84 = "Transform (Modify) Bone - Bone: dyn_jacket_r_MSTR"(ComponentPose=@n83, Alpha=@n81, Translation="0,0,0", Rotation="10,0,0", Scale="1,1,1") {
    rotationMode=BMM_Additive
    rotationSpace=BCS_BoneSpace
    translationMode=BMM_Additive
    translationSpace=BCS_BoneSpace
    alphaBoolBlend {blendOption=Linear}
    alphaScaleBiasClamp {bInterpResult=true}
    boneToModify {boneName=dyn_jacket_r_MSTR}
    rotation {pitch=10}
   }
   n85 = "Transform (Modify) Bone - Bone: dyn_jacket_bk_r_MSTR"(ComponentPose=@n84, Alpha=@n81, Translation="0,0,0", Rotation="5,0,0", Scale="1,1,1") {
    rotationMode=BMM_Additive
    rotationSpace=BCS_BoneSpace
    translationMode=BMM_Additive
    translationSpace=BCS_BoneSpace
    alphaBoolBlend {blendOption=Linear}
    alphaScaleBiasClamp {bInterpResult=true}
    boneToModify {boneName=dyn_jacket_bk_r_MSTR}
    rotation {pitch=5}
   }
   n86 = "Transform (Modify) Bone - Bone: dyn_jacket_bk_MSTR"!(ComponentPose=@n85, bAlphaBoolEnabled:bool=True, Translation="0,0,0", Rotation="0.000,0.000,20", Scale="1,1,1") -> (Pose) {
    alphaInputType=Bool
    rotationMode=BMM_Additive
    alphaBoolBlend {blendInTime=0.2, blendOption=Linear, blendOutTime=0.5}
    boneToModify {boneName=dyn_jacket_bk_MSTR}
    rotation {roll=20}
   }
   n87 = "Transform (Modify) Bone - Bone: dyn_jacket_bk_r_MSTR"!(ComponentPose=@n86, bAlphaBoolEnabled:bool=True, Translation="0,0,0", Rotation="0.000,0.000,15", Scale="1,1,1") -> (Pose) {
    alphaInputType=Bool
    rotationMode=BMM_Additive
    alphaBoolBlend {blendInTime=0.2, blendOption=Linear, blendOutTime=0.5}
    boneToModify {boneName=dyn_jacket_bk_r_MSTR}
    rotation {roll=15}
   }
   n88 = "Transform (Modify) Bone - Bone: dyn_jacket_bk_l_MSTR"!(ComponentPose=@n87, bAlphaBoolEnabled:bool=True, Translation="0,0,0", Rotation="0.000,0.000,15", Scale="1,1,1") -> (Pose) {
    alphaInputType=Bool
    rotationMode=BMM_Additive
    alphaBoolBlend {blendInTime=0.2, blendOption=Linear, blendOutTime=0.5}
    boneToModify {boneName=dyn_jacket_bk_l_MSTR}
    rotation {roll=15}
   }
   Save cached pose 'constraints'(Pose=Component To Local(ComponentPose=@n88)) {cacheName=constraints}
   Get GravityOverride()
  ]}
 EventGraph {nodes [
   Event BlueprintUpdateAnimation()
   > Get!(execute:Exec) -> (then:Exec, else:Exec, Skel_Mesh:Skeletal Mesh Component Object Reference)
   > n1 = Sequence!(execute) -> (then_0, then_1, then_2:Exec, then_3:Exec, then_4:Exec, then_5:Exec, then_6:Exec, then_7:Exec, then_8:Exec)
   n2 = Get Skel_Mesh()
   Set thigh_rot_l(
    execute=[@n1.then_0, @n18.else]
    thigh_rot_l=GetAlphaFromReferenceTransformDelta(SkeletalMeshComponent=@n2, BoneName=thigh_l, BaseName=pelvis, InputRange=(X=-20,Y=-100), TwistAxis="1,0,0")
   )
   > Set thigh_rot_r(thigh_rot_r=GetAlphaFromReferenceTransformDelta(SkeletalMeshComponent=@n2, BoneName=thigh_r, BaseName=pelvis, InputRange=(X=-20,Y=-100), TwistAxis="1,0,0"))
   n3 = Get Skel_Mesh()
   > Set spine Left(spine Left:double=GetAlphaFromReferenceTransformDelta(SkeletalMeshComponent=@n3, BoneName=head, BaseName=pelvis, DeltaComponentType=EulerY, InputRange=(X=-5,Y=-50), TwistAxis="1,0,0")) -> (Output_Get:double)
   > Set spine right(spine right:double=GetAlphaFromReferenceTransformDelta(SkeletalMeshComponent=@n3, BoneName=spine_05, BaseName=pelvis, InputRange=(X=0,Y=40), TwistAxis="1,0,0")) -> (Output_Get:double)
   > Set spine back(spine back:double=GetAlphaFromReferenceTransformDelta(SkeletalMeshComponent=@n3, BoneName=head, BaseName=pelvis, InputRange=(X=0,Y=-50), TwistAxis="1,0,0")) -> (Output_Get:double)
   > n4 = PlayerMovement(AnimInstance=Self-Reference(), XDotProductCoefficient=1, YDotProductCoefficient=1, ZDotProductCoefficient=1)
   > Set Move lft(Move lft:double=MapRangeClamped(Value=@n4, InRangeA=0, InRangeB=-0.4)) -> (Output_Get:double)
   > Set Move right(Move right:double=MapRangeClamped(Value=@n4, InRangeA=0, InRangeB=0.4)) -> (Output_Get:double)
   > Set Move bwd(Move bwd:double=MapRangeClamped(Value=@n4.YMovement, InRangeA=0, InRangeB=-0.7)) -> (Output_Get:double)
   n5 = Gravity Override State Function DEPRECATED(
    execute=@n1.then_1
    CharacterPartAnimInstance=Self-Reference()
    GameTimeInSeconds=GetGameTimeInSeconds()
    GravityOverrideParams=Get Gravity Override Params()
   )
   > Set GravityOverride(GravityOverride=@n5)
   n6 = Cast To FortPlayerAnimInstance(execute=@n1.then_2, Object=GetAnimInstance(self=GetOwnerSkeletalMeshComponent()))
   > Set weapon(weapon:String=Enum to String!(Enumerator:EFortWeaponCoreAnimation Enum=Get WeaponCoreAnim(self=@n6)) -> (ReturnValue:String)) -> (Output_Get:String)
   n7 = Get weapon() -> (weapon:String)
   n8 = Branch(execute=@n1.then_3, Condition=AND Boolean(A="Equal Exactly (String)"(A=@n7, B=Assault Rifle), B=Get bIsCrouching()))
   > Set Rifle(Rifle:bool=true) -> (Output_Get:bool)
   Set Rifle(execute=@n8.else, Rifle=false)
   n9 = Branch(execute=@n1.then_4, Condition=Get bIsCrouching())
   > n10 = Branch(Condition=OR Boolean(A=Get bIsRelaxedLevel1() -> (bIsRelaxedLevel1:bool), B=Get bIsRelaxedLevel2() -> (bIsRelaxedLevel2:bool)))
   > Set Rifle_crouch(Rifle_crouch:bool=false) -> (Output_Get:bool)
   n11 = Branch(execute=@n10.else, Condition="Equal Exactly (String)"(A=Get weapon(), B=Assault Rifle))
   > Set Rifle_crouch(Rifle_crouch=true)
   Set Rifle_crouch(execute=@n11.else, Rifle_crouch=false)
   Set Rifle_crouch(execute=@n9.else, Rifle_crouch=false)
   n12 = Branch(execute=@n1.then_5, Condition=Get bIsCrouching())
   n13 = Get weapon()
   > Branch(Condition=AND Boolean(A="Equal Exactly (String)"(A=@n13, B=Assault Rifle), B=Get bIsTargeting()))
   > Set Rifle_crouch(Rifle_crouch=true)
   Set Rifle_crouch(execute=@n12.else, Rifle_crouch=false)
   Set Pickaxe anim(
    execute=@n1.then_6
    Pickaxe anim:double=GetAlphaFromReferenceTransformDelta(SkeletalMeshComponent=Get Skel_Mesh(), BoneName=Head, BaseName=pelvis, InputRange=(X=30,Y=50), TwistAxis="1,0,0")
   ) -> (Output_Get:double)
   n14 = ComplexSimSpaceFunction DEPRECATED!(
    execute=@n1.then_7
    CharacterPartAnimInstance:Custom Character Part Anim Instance Object Reference=Self-Reference()
    InterpSpeedMax:double=9
    PawnSpeedMax:double=20
    RInterpErrorTolerance:double=1
    DeltaTime:double=0
    DebugDraw:bool=false
    sim_space_rotation_in:Rotator=Get sim_space_rotation() -> (sim_space_rotation:Rotator)
    inComponentSpaceIn:bool=Get InComponentSpace() -> (InComponentSpace:bool)
    inComponentSpaceLastFrameIn:bool=Get InComponentSpaceLastFrame() -> (InComponentSpaceLastFrame:bool)
    lockedOnSimSpaceIn:bool=Get LockedOnSimSpace() -> (LockedOnSimSpace:bool)
    InterpSpeedIn:double=Get InterpSpeed() -> (InterpSpeed:double)
   ) -> (
    then
    sim_space_rotation:Rotator
    inComponentSpaceOut:bool
    inComponentSpaceLastFrameOut:bool
    lockedOnSimSpaceOut:bool
    InterpSpeedOut:double
   )
   > Set sim_space_rotation(sim_space_rotation:Rotator=@n14) -> (Output_Get:Rotator)
   > Set InComponentSpace(InComponentSpace:bool=@n14.inComponentSpaceOut) -> (Output_Get:bool)
   > Set InComponentSpaceLastFrame(InComponentSpaceLastFrame:bool=@n14.inComponentSpaceLastFrameOut) -> (Output_Get:bool)
   > Set LockedOnSimSpace(LockedOnSimSpace:bool=@n14.lockedOnSimSpaceOut) -> (Output_Get:bool)
   > Set InterpSpeed(InterpSpeed:double=@n14.InterpSpeedOut) -> (Output_Get:double)
   n15 = GetPartSkeletalMeshComponent()
   Is Valid(exec=@n1.then_8, InputObject=@n15)
   > SetVectorParameterValueOnMaterials!(
    execute
    self:Mesh Component Object Reference=@n15
    ParameterName:Name=DO NOT OVERRIDE-Virtual-Pivot-Ch
    ParameterValue:Vector=TransformLocation!(T:"Transform (by ref)"=GetSocketTransform(self=@n15, InSocketName=spine_05, TransformSpace=RTS_World), Location:Vector="-5, 0,0") -> (ReturnValue:Vector)
   ) -> (then)
   > SetScalarParameterValueOnMaterials!(execute, self:Mesh Component Object Reference=@n15, ParameterName:Name=DO NOT OVERRIDE-Character-Position-BP-Switch, ParameterValue:float=1) -> (then)
   Event BlueprintInitializeAnimation()
   > Set Skel_Mesh(Skel_Mesh=GetOwnerSkeletalMeshComponent())
   Event OnCharacterCustomizationLoaded()
   > n16 = Sequence!(execute) -> (then_0, then_1)
   n17 = CheckForBackbling()
   > Set BackBling_Alpha(BackBling_Alpha:double=SelectFloat(A=0, B=1, bPickA=@n17)) -> (Output_Get:double)
   > Set RigidsAcc(RigidsAcc:Vector=SelectVector(A="0.15,0.1,0.1", B="0.25,0.2,0.2", bPickA=@n17)) -> (Output_Get:Vector)
   > Set RigidsVel(RigidsVel:Vector=SelectVector(A="0.1,0.1,0.1", B="0.25,0.2,0.2", bPickA=@n17)) -> (Output_Get:Vector)
   ResetDynamics(execute=@n16.then_1, self=Self-Reference(), InTeleportType=None)
   n18 = Branch(Condition=Get bIsPlayingMeleeAnim())
   > Set thigh_rot_l(thigh_rot_l=0)
   > Set thigh_rot_r(thigh_rot_r=0)
   Set Rifle(Rifle=true)
   Set Rifle(Rifle=false)
  ]}
]
variables [
 BackBling_Alpha {type=number}
 Gravity Override Params {
  type=GravityOverrideParamsStruct
  default {
   skydive_DivingDown {gravityOverride=<0 0 0>, windAmplitude=<100 -250 200>, windFrequency=<0.75 3.5 4>, windOffset=<0 -300 200>}
   skydive_FloatingDown {gravityOverride=<0 0 0>, windAmplitude=<100 -250 200>, windFrequency=<0.75 3.5 4>, windOffset=<0 -300 200>}
   skydive_GoingUp {gravityOverride=<0 0 100>, windAmplitude=<100 -250 200>, windFrequency=<0.75 3.5 4>, windOffset=<0 -300 200>}
   skydive_Parachuting {gravityOverride=<0 0 -270>, windAmplitude=<100 -250 200>, windFrequency=<0.75 3.5 4>, windOffset=<0 -300 200>}
  }
 }
 GravityOverride {type=Vector}
 InComponentSpace {type=boolean}
 InComponentSpaceLastFrame {type=boolean}
 InterpSpeed {type=number}
 LockedOnSimSpace {type=boolean}
 Move bwd {type=number}
 Move lft {type=number}
 Move right {type=number}
 Pickaxe anim {type=number}
 Rifle {type=boolean}
 Rifle_crouch {type=boolean}
 RigidsAcc {type=Vector}
 RigidsVel {type=Vector}
 sim_space_rotation {type=Rotator}
 Skel_Mesh {type=SkeletalMeshComponent}
 spine back {type=number}
 spine Left {type=number}
 spine right {type=number}
 thigh_rot_l {type=number}
 thigh_rot_r {type=number}
 weapon {type=string, default=""}
]
