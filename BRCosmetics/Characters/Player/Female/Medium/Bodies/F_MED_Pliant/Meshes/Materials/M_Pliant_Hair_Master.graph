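# Master hair material for the Pliant character's skeletal meshes, adding recolorable hair to the shared hair shader.
# FN_Hair supplies the base attributes; its desaturated, power-adjusted color multiplies the primary color, then blends in through an invertible channel-selected mask.
# An optional static-switch ClothFuzz layer follows, hence the expression-driven Default Lit or Cloth shading model.
# The result chains through animated-hair, tech-art override, de-rez transition and character opacity functions.
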Material /BRCosmetics/Characters/Player/Female/Medium/Bodies/F_MED_Pliant/Meshes/Materials/M_Pliant_Hair_Master
graph {
 properties {
  bUseMaterialAttributes=true
  bUsedWithMeshDeformer=true
  bUsedWithSkeletalMesh=true
  pixelDepthOffsetMode=PDOM_Legacy
  shadingModel=MSM_FromMaterialExpression
  usedShadingModels=Default Lit | Cloth
 }
 nodes [
  n1 = ScalarParameter() {defaultValue=1, group=Elastic, parameterName=HairClothFuzzAlpha, sliderMax=1}
  n2 = TextureSampleParameter2D() {
   group=Hair
   parameterName=M
   samplerType=SAMPLERTYPE_Masks
   texture=/Game/Accessories/Hats/M_MED_Assassin_Suit_FaceAcc/Textures/T_Assassin_Suit_Hair_M
  }
  n3 = VectorParameter() {group=Pliant, parameterName=Pliant_PrimaryColor, defaultValue=<rgba 0.9271 0.0922 0.5072 1>}
  n4 = ScalarParameter() {defaultValue=1, group=Pliant, parameterName=Pliant_HairBrightnessBoost}
  n5 = GetMaterialAttributes(None=MaterialFunctionCall() {materialFunction=/Game/Packages/Fortress_SharedMaterials/Base_Material_Functions/FN_Hair}.Result) {attributeGetTypes [#1]}
  n6 = Power(Base=@n5.Base Color, Exp=ScalarParameter() {defaultValue=1, group=Pliant, parameterName=Pliant_HairTintPow, sliderMax=2})
  n7 = Multiply(
   A=Saturate(None=Multiply(A=@n3.RGB, B=@n4))
   B=Power(Base=Desaturation(None=@n6), Exp=ScalarParameter() {defaultValue=0.2, group=Pliant, parameterName=Pliant_HairTintDarkenPow, sliderMax=2}) {constExponent=0.1}
  )
  n8 = TextureSampleParameter2D() {
   group=Pliant
   parameterName=Pliant_TintMask_Hair
   texture=/BRCosmetics/Characters/Player/Male/Medium/Bodies/M_MED_Pliant/Textures/M_MED_Pliant_Hair_FX
  }
  n9 = ChannelMaskParameter(None=@n8.RGB) {group=Pliant, parameterName=Pliant_TintableHairChannel}
  n10 = OneMinus(None=@n9.R)
  n11 = ScalarParameter() {group=Pliant, parameterName=Pliant_TintMask_Hair_Inverted, sliderMax=2}
  n12 = Clamp(None=Power(
    Base=LinearInterpolate(A=@n9.R, B=@n10, Alpha=@n11)
    Exp=ScalarParameter() {defaultValue=1, group=Pliant, parameterName=Pliant_HairTintPow, sliderMax=2}
   ))
  n13 = SetMaterialAttributes!(
   MaterialAttributes=@n5.MaterialAttributes
   Base Color=LinearInterpolate(
    A=@n5.Base Color
    B=LinearInterpolate(A=@n5.Base Color, B=@n7, Alpha=@n12)
    Alpha=ScalarParameter() {defaultValue=1, group=Pliant, parameterName=ColorHair}
   )
  ) {attributeSetTypes [#1]}
  n14 = MaterialFunctionCall!(
   GMap
   Material=StaticSwitchParameter(
    True=MaterialFunctionCall!(Alpha=@n1, Gmap=@n2.G, Material=@n13) {materialFunction=/Game/Weapons/WeaponSkins/MaterialFunction/ClothFuzz}.Result
    False=@n13
   ) {group=Skin FX, parameterName=Use Cloth fuzz}
  ) {materialFunction=/Game/Weapons/WeaponSkins/MaterialFunction/MF_AnimatedHair}
  MaterialFunctionCall!(Material=MaterialFunctionCall!(
    Fresnel=Fresnel()
    Material=MaterialFunctionCall(In=@n14.Result) {materialFunction=&MF_TechArtOverrides}.Result
   ) {materialFunction=/Game/Packages/Fortress_SharedMaterials/Base_Material_Functions/CharacterDeRez_Transition}.Result) {materialFunction=/Game/Packages/Fortress_SharedMaterials/Base_Material_Functions/BaseCharacterShaderFunctions/BaseFN_CharacterOpacity}
 ]
}
