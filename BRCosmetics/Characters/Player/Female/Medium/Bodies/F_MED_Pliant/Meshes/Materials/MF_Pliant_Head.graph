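# Material function for the female Pliant character head that tints the scalp and hair area of incoming Material Attributes with the primary color.
# The base color is desaturated, power-adjusted, then multiplied by the brightness-boosted, saturated tint.
# A selectable tint-mask channel, defaulting to green of the male Pliant head FX texture, is power-adjusted and blends this into BaseColor.
# The ScalpTintSaturation lerp is not connected to the output.
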MaterialFunction /BRCosmetics/Characters/Player/Female/Medium/Bodies/F_MED_Pliant/Meshes/Materials/MF_Pliant_Head
graph {nodes [
  n1 = GetMaterialAttributes(None=FunctionInput() {inputType=FunctionInput_MaterialAttributes}) {attributeGetTypes [#1]}
  n2 = VectorParameter() {group=Pliant, parameterName=Pliant_PrimaryColor, defaultValue=<rgba 0.9271 0.0922 0.5072 1>}
  n3 = ScalarParameter() {defaultValue=1, group=Pliant, parameterName=Pliant_HairBrightnessBoost}
  n4 = Power(
   Base=Desaturation(None=@n1.BaseColor)
   Exp=ScalarParameter() {defaultValue=1, group=Pliant, parameterName=Pliant_HairTintDarkenPow, sliderMax=2}
  ) {constExponent=0.1}
  n5 = Multiply(A=Saturate(None=Multiply(A=@n2.RGB, B=@n3)), B=@n4)
  n6 = TextureSampleParameter2D() {
   group=Pliant
   parameterName=Pliant_TintMask_Head
   texture=/BRCosmetics/Characters/Player/Male/Medium/Bodies/M_MED_Pliant/Textures/M_MED_Pliant_Head_FX
  }
  n7 = ChannelMaskParameter(None=@n6.RGB) {group=Pliant, maskChannel=Green, parameterName=Pliant_TintableScalpChannel, defaultValue=<rgba 0 1 0 0>}
  n8 = Clamp(None=Power(Base=@n7.R, Exp=ScalarParameter() {defaultValue=1, group=Pliant, parameterName=Pliant_ScalpTintPow, sliderMax=2}))
  FunctionOutput(None=SetMaterialAttributes!(MaterialAttributes=@n1.MaterialAttributes, Normal, BaseColor=LinearInterpolate(A=@n1.BaseColor, B=@n5, Alpha=@n8), WorldPositionOffset) {attributeSetTypes [#2, #1, #3]})
  LinearInterpolate(A=@n4, Alpha=ScalarParameter() {defaultValue=1, group=Pliant, parameterName=Pliant_ScalpTintSaturation, sliderMax=2})
 ]}
