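# Master material for the head of the Pliant female player character cosmetic, for skeletal meshes and mesh deformers.
# It chains shared character functions through material attributes: PreFX feeds CharacterSkinFX, then the head-specific MF_Pliant_Head, then PostFX.
# PreFX also supplies inputs directly to PostFX, bypassing the middle stages.
# Shading model comes from the material expression, with a hero skin subsurface profile assigned.
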Material /BRCosmetics/Characters/Player/Female/Medium/Bodies/F_MED_Pliant/Meshes/Materials/M_Pliant_Head_Master
graph {
 properties {
  bUseMaterialAttributes=true
  bUsedWithMeshDeformer=true
  bUsedWithSkeletalMesh=true
  pixelDepthOffsetMode=PDOM_Legacy
  refractionMethod=RM_IndexOfRefraction
  shadingModel=MSM_FromMaterialExpression
  subsurfaceProfile=&SS_HeroSkin_02
  usedShadingModels=Default Lit | Subsurface | Clear Coat | Subsurface Profile | Cloth | Eye
 }
 nodes [
  n1 = MaterialFunctionCall() {materialFunction=&BaseFN_CharacterFunction_PreFX}
  MaterialFunctionCall(
   AO=@n1.AO
   Distance Blend=@n1.AO
   Fresnel=@n1.AO
   Gmap=@n1.AO
   Pre Skinned Local Position=@n1.AO
   Material=MaterialFunctionCall!(In=MaterialFunctionCall(PreskinnedPosition=@n1.Pre skinned Local Position, MaterialAttributes=@n1.Pre skinned Local Position, AO=@n1.Pre skinned Local Position) {materialFunction=&CharacterSkinFX}.Result) {materialFunction=&/MF_Pliant_Head}.Result
  ) {materialFunction=&BaseFN_CharacterFunction_PostFX}
 ]
}
