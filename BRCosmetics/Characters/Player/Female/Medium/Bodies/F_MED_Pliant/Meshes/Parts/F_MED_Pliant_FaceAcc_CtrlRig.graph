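# Control Rig for the F_MED_Pliant face accessory that nudges braid bones in response to head rotation.
# It converts the head's rotation relative to spine_05 to Euler, then remaps Z through two clamped ranges into Y translation offsets, both applied to dyn_braid_r_02.
# A second Modify Transforms step starting at dyn_braid_r_01 has no driven inputs, and two Euler-to-quaternion nodes are unconnected.
# The hierarchy holds both braid chains, but the wired offsets target the right one.
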ControlRigBlueprint /BRCosmetics/Characters/Player/Female/Medium/Bodies/F_MED_Pliant/Meshes/Parts/F_MED_Pliant_FaceAcc_CtrlRig
graphs [
 RigVMFunctionLibrary {links [], nodes []}
 RigVMModel {nodes [
   RigUnit_BeginExecution = &RigUnit_BeginExecution()
   n1 = &RigVMFunction_MathQuaternionToEuler(
    RotationOrder=ZYX
    Value=&RigUnit_GetRelativeTransformForItem(Child=(Type=Bone,Name="head"), Parent=(Type=Bone,Name="spine_05")).RelativeTransform.Rotation
   )
   &RigUnit_ModifyTransforms(
    ItemToModify="((Item=(Type=Bone,Name=\"dyn_braid_r_02\"),Transform=(Rotation=(X=0,Y=0,Z=0,W=1),Translation=(X=0,Y=0,Z=0),Scale3D=(X=1,Y=1,Z=1))),(Item=(Type=Bone,Name=\"dyn_braid_r_02\"),Transform=(Rotation=(X=0,Y=0,Z=0,W=1),Translation=(X=0,Y=0,Z=0),Scale3D=(X=1,Y=1,Z=1))) ...(truncated)"
    Weight=1
    ItemToModify.0.Transform.Translation.Y=&RigVMFunction_MathFloatRemap(SourceMaximum=-7, SourceMinimum=-40, TargetMaximum=0, TargetMinimum=3, bClamp=true, Value=@n1.Result.Z).Result
    ItemToModify.1.Transform.Translation.Y=&RigVMFunction_MathFloatRemap(SourceMaximum=20, SourceMinimum=-5, TargetMaximum=0, TargetMinimum=-2, bClamp=true, Value=@n1.Result.Z).Result
    ExecutePin=@RigUnit_BeginExecution.ExecutePin
   )
   > &RigUnit_ModifyTransforms(
    ItemToModify="((Item=(Type=Bone,Name=\"dyn_braid_r_01\"),Transform=(Rotation=(X=0,Y=0,Z=0,W=1),Translation=(X=0,Y=0,Z=0),Scale3D=(X=1,Y=1,Z=1))),(Item=(Type=Bone,Name=\"None\"),Transform=(Rotation=(X=0,Y=0,Z=0,W=1),Translation=(X=0,Y=0,Z=0),Scale3D=(X=1,Y=1,Z=1))),(Item=(Ty ...(truncated)"
    Weight=1
   )
   &RigVMFunction_MathQuaternionFromEuler(Euler=(X=0,Y=0,Z=0))
   &RigVMFunction_MathQuaternionFromEuler(Euler=(X=0,Y=0,Z=-1))
  ]}
]
hierarchy [
 root {loc=<0 0 0>, rot=<0 0 0 1>}
 pelvis {parent=root, loc=<0 -0.5301 98.0571>, rot=<0.0127 0.707 -0.0127 -0.707>}
 spine_01 {parent=pelvis, loc=<6.1208 -0.9328 0>, rot=<0 0 0.034 -0.9994>}
 spine_02 {parent=spine_01, loc=<6.5996 -0.2095 0>, rot=<0 0 0.0238 -0.9997>}
 spine_03 {parent=spine_02, loc=<6.4914 -0.1237 0>, rot=<0 0 0.0149 -0.9999>}
 spine_04 {parent=spine_03, loc=<6.9598 0.0794 0>, rot=<0 0 -0.0132 -0.9999>}
 spine_05 {parent=spine_04, loc=<6.9987 0 0>, rot=<0 0 -0.0015 -1>}
 clavicle_l {parent=spine_05, loc=<5.4831 -0.1053 -3.2946>, rot=<0.0465 -0.7289 -0.009 -0.6829>}
 clavicle_r {parent=spine_05, loc=<5.4831 -0.1053 3.2946>, rot=<0.7289 0.0465 -0.6829 0.009>}
 neck_01 {parent=spine_05, loc=<11.3138 0.1529 0>, rot=<0 0 -0.2003 -0.9797>}
 neck_02 {parent=neck_01, loc=<4.5627 0 0>, rot=<0 0 0.0504 -0.9987>}
 head {parent=neck_02, loc=<5.3914 0 0>, rot=<0 0 0.115 -0.9934>}
 dyn_head_r_lookat {parent=head, loc=<0 0 0>, rot=<0 0 0 -1>}
 dyn_braid_r_MSTR {parent=dyn_head_r_lookat, loc=<3.2216 -5.7411 4.8143>, rot=<0.1081 0.6632 0.7342 0.0971>}
 dyn_braid_r_01 {parent=dyn_braid_r_MSTR, loc=<1.1253 0 0>, rot=<0 0 0 1>}
 dyn_braid_r_02 {parent=dyn_braid_r_01, loc=<9.1963 0 0>, rot=<-0.0738 0.1996 0.0405 -0.9762>}
 dyn_braid_r_03 {parent=dyn_braid_r_02, loc=<5.2297 0 0>, rot=<-0.0312 0.0298 0.0281 -0.9987>}
 dyn_braid_r_04 {parent=dyn_braid_r_03, loc=<5.5542 0 0>, rot=<0.0228 0.0037 -0.022 -0.9995>}
 dyn_braid_r_05 {parent=dyn_braid_r_04, loc=<4.1124 0 0>, rot=<-0.0033 -0.0685 0.2417 -0.9679>}
 dyn_braid_r_constraint {parent=dyn_braid_r_05, loc=<5.7592 0 0>, rot=<-0.3646 0.2631 0.4191 -0.7888>}
 dyn_head_l_lookat {parent=head, loc=<0 0 0>, rot=<0 0 0 -1>}
 dyn_braid_l_MSTR {parent=dyn_head_l_lookat, loc=<3.224 -5.7332 -4.8164>, rot=<-0.0948 -0.7335 -0.6647 -0.1052>}
 dyn_braid_l_01 {parent=dyn_braid_l_MSTR, loc=<1.1253 0 0>, rot=<0 0 0 1>}
 dyn_braid_l_02 {parent=dyn_braid_l_01, loc=<8.7185 0 0>, rot=<0.0553 0.2109 -0.0306 -0.9755>}
 dyn_braid_l_03 {parent=dyn_braid_l_02, loc=<5.9468 0 0>, rot=<0.0607 0.0041 -0.0535 -0.9967>}
 dyn_braid_l_04 {parent=dyn_braid_l_03, loc=<5.3021 0 0>, rot=<-0.0053 0.0158 0.0049 -0.9998>}
 dyn_braid_l_05 {parent=dyn_braid_l_04, loc=<4.2767 0 0>, rot=<-0.111 -0.0668 -0.1998 -0.9712>}
 dyn_braid_l_constraint {parent=dyn_braid_l_05, loc=<6.8027 0 0>, rot=<-0.2251 0.4247 0.7532 -0.449>}
 dyn_clavicle_l {parent=spine_05, loc=<5.4831 -0.1053 -3.2946>, rot=<0.0465 -0.7289 -0.009 -0.6829>}
 dyn_clavicle_r {parent=spine_05, loc=<5.4831 -0.1053 3.2946>, rot=<0.7289 0.0465 -0.6829 0.009>}
 dyn_spine_04_r_lookat {parent=spine_04, loc=<4.3771 11.8074 4.156>, rot=<0 0 0 1>}
 dyn_spine_04_l_lookat {parent=spine_04, loc=<4.3771 11.8074 -4.1558>, rot=<0 0 0 1>}
]
