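# Physics asset for the face accessory mesh of the medium female Pliant player body, covering left and right braid bone chains.
# Each braid is a five-capsule chain under a small master body linked to neck_02, with tight twist limits and swing limits that generally loosen past the root.
# Larger capsules on the upper spine, neck and clavicles give the body's shape, joined to each other by constraints with free swing and twist.
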PhysicsAsset /BRCosmetics/Characters/Player/Female/Medium/Bodies/F_MED_Pliant/Meshes/Parts/F_MED_Pliant_FaceAcc_Physics
bodies {
 dyn_braid_l_01 [{shapeType=Capsule, length=6.2652, radius=1.986, center=<3 0 0>, rotation=<rot 90 0 0>}]
 dyn_braid_l_02 [{shapeType=Capsule, length=4.4753, radius=1.6554, center=<2.9491 0 0>, rotation=<rot 90 0 0>}]
 dyn_braid_l_03 [{shapeType=Capsule, length=1.1114, radius=2.0648, shapeName=dyn_braid_l_03_capsule, center=<2.5 0.5 1.5>, rotation=<rot 90 0 0>}]
 dyn_braid_l_04 [{shapeType=Capsule, length=1.1114, radius=2.0648, shapeName=dyn_braid_l_03_capsule, center=<2.5 0.5 1.5>, rotation=<rot 90 0 0>}]
 dyn_braid_l_05 [{shapeType=Capsule, length=1.1114, radius=2.0648, shapeName=dyn_braid_l_03_capsule, center=<2.5 0.5 1.5>, rotation=<rot 90 0 0>}]
 dyn_braid_l_MSTR [{shapeType=Capsule, length=0.505, radius=0.505}]
 dyn_braid_r_01 [{shapeType=Capsule, length=7.654, radius=1.8467, center=<3 0 0>, rotation=<rot 90 0 0>}]
 dyn_braid_r_02 [{shapeType=Capsule, length=4.4051, radius=1.582, center=<2.6999 0 0>, rotation=<rot 90 0 0>}]
 dyn_braid_r_03 [{shapeType=Capsule, length=1.1114, radius=2.0648, shapeName=dyn_braid_r_03_capsule, center=<2.5 -0.5 1.5>, rotation=<rot 90 0 0>}]
 dyn_braid_r_04 [{shapeType=Capsule, length=1.1114, radius=2.0648, shapeName=dyn_braid_r_03_capsule, center=<2.5 -0.5 1.5>, rotation=<rot 90 0 0>}]
 dyn_braid_r_05 [{shapeType=Capsule, length=1.1114, radius=2.0648, shapeName=dyn_braid_r_03_capsule, center=<2.5 -0.5 1.5>, rotation=<rot 90 0 0>}]
 dyn_braid_r_MSTR [{shapeType=Capsule, length=0.505, radius=0.505}]
 dyn_clavicle_l [{shapeType=Capsule, length=23.5125, radius=6.1813, center=<7.6803 6.9416 -0.0787>, rotation=<rot -4.3286 2.5051 129.9028>}]
 dyn_clavicle_r [{shapeType=Capsule, length=22.8071, radius=6.1813, center=<-7.836 -7.7478 -0.3481>, rotation=<rot -4.3285 2.5051 129.9006>}]
 neck_01 [{shapeType=Capsule, length=6.4408, radius=3.1464, rotation=<rot -90 5.0511 -5.0511>}]
 neck_02 [{shapeType=Capsule, length=20.8544, radius=6.3044, shapeName=neck_02_capsule, center=<-3.6476 -8.9877 0>, rotation=<rot 0 -90 0>}]
 spine_04 [{shapeType=Capsule, length=9.6437, radius=10.5062, center=<1.9805 1.1614 0>, rotation=<rot 0 -90 0>}]
 spine_05 [{shapeType=Capsule, length=17.6847, radius=9.1856, center=<0.7116 -2.2634 0>, rotation=<rot 0 -8.0782 0>}]
}
constraints [
 {
  bone1Name=neck_01
  bone2Name=spine_05
  swing1LimitDegrees=45
  swing1Motion=Free
  swing2LimitDegrees=45
  swing2Motion=Free
  twistLimitDegrees=45
  twistMotion=Free
 }
 {bone1Name=dyn_braid_r_05, bone2Name=dyn_braid_r_04, swing1LimitDegrees=20, swing2LimitDegrees=35, twistLimitDegrees=5}
 {bone1Name=dyn_braid_r_03, bone2Name=dyn_braid_r_02, swing1LimitDegrees=25, swing2LimitDegrees=35, twistLimitDegrees=5}
 {bone1Name=dyn_braid_r_04, bone2Name=dyn_braid_r_03, swing1LimitDegrees=20, swing2LimitDegrees=35, twistLimitDegrees=5}
 {bone1Name=dyn_braid_l_05, bone2Name=dyn_braid_l_04, swing1LimitDegrees=20, swing2LimitDegrees=35, twistLimitDegrees=5}
 {bone1Name=dyn_braid_l_04, bone2Name=dyn_braid_l_03, swing1LimitDegrees=20, swing2LimitDegrees=35, twistLimitDegrees=5}
 {bone1Name=dyn_braid_l_03, bone2Name=dyn_braid_l_02, swing1LimitDegrees=20, swing2LimitDegrees=35, twistLimitDegrees=5}
 {bone1Name=dyn_braid_r_02, bone2Name=dyn_braid_r_01, swing1LimitDegrees=10, swing2LimitDegrees=20, twistLimitDegrees=5}
 {bone1Name=dyn_braid_l_02, bone2Name=dyn_braid_l_01, swing1LimitDegrees=15, swing2LimitDegrees=20, twistLimitDegrees=5}
 {
  bone1Name=spine_05
  bone2Name=spine_04
  swing1LimitDegrees=45
  swing1Motion=Free
  swing2LimitDegrees=45
  swing2Motion=Free
  twistLimitDegrees=45
  twistMotion=Free
 }
 {bone1Name=dyn_braid_r_01, bone2Name=dyn_braid_r_MSTR, swing1LimitDegrees=10, swing2LimitDegrees=10, twistLimitDegrees=3}
 {
  bone1Name=dyn_braid_r_MSTR
  bone2Name=neck_02
  swing1LimitDegrees=45
  swing1Motion=Free
  swing2LimitDegrees=45
  swing2Motion=Free
  twistLimitDegrees=45
  twistMotion=Free
 }
 {
  bone1Name=dyn_braid_l_MSTR
  bone2Name=neck_02
  swing1LimitDegrees=45
  swing1Motion=Free
  swing2LimitDegrees=45
  swing2Motion=Free
  twistLimitDegrees=45
  twistMotion=Free
 }
 {bone1Name=dyn_braid_l_01, bone2Name=dyn_braid_l_MSTR, swing1LimitDegrees=10, swing2LimitDegrees=10, twistLimitDegrees=3}
 {
  bone1Name=dyn_clavicle_r
  bone2Name=spine_05
  swing1LimitDegrees=45
  swing1Motion=Free
  swing2LimitDegrees=45
  swing2Motion=Free
  twistLimitDegrees=45
  twistMotion=Free
 }
 {
  bone1Name=dyn_clavicle_l
  bone2Name=spine_05
  swing1LimitDegrees=45
  swing1Motion=Free
  swing2LimitDegrees=45
  swing2Motion=Free
  twistLimitDegrees=45
  twistMotion=Free
 }
 {
  bone1Name=neck_02
  bone2Name=neck_01
  swing1LimitDegrees=45
  swing1Motion=Free
  swing2LimitDegrees=45
  swing2Motion=Free
  twistLimitDegrees=45
  twistMotion=Free
 }
]
