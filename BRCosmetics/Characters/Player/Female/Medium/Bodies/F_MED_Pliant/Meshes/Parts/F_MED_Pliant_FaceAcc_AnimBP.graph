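# Post-process style secondary-motion AnimBP for the F_MED_Pliant face accessory mesh, animating the braids and collar on the attached parent's copied pose.
# It scales the collar and partially aims both braid master bones at spine look-at bones before a neck-relative RigidBody simulation.
# Weighted constraints then pull the braid tips toward those look-at bones, and a Control Rig finishes the pose.
# Unconnected AnimDynamics, Look At, Constraint and Transform nodes remain, and the EventGraph is empty.
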AnimBlueprint /BRCosmetics/Characters/Player/Female/Medium/Bodies/F_MED_Pliant/Meshes/Parts/F_MED_Pliant_FaceAcc_AnimBP
target_skeleton=&/F_MED_Pliant_FaceAcc_Skeleton
graphs [
 AnimGraph {nodes [
   n1 = Anim Dynamics - Bone: dyn_braid_l_01(Alpha=1, ExternalForce="0,0,0") {
    angularDampingOverride=0.998
    angularSpringConstant=150
    bAngularSpring=true
    bGravityOverrideInSimSpace=true
    bOverrideAngularDamping=true
    bUseGravityOverride=true
    simulationSpace=RootRelative
    alphaBoolBlend {blendOption=Linear}
    boundBone {boneName=dyn_braid_l_01}
    chainEnd {boneName=dyn_braid_l_05}
    componentAppliedLinearAccClamp=<3000 3000 3000>
    componentLinearAccScale=<0.3 0.3 0.3>
    componentLinearVelScale=<0.3 0.3 0.3>
    gravityOverride {z=-980}
    physicsBodyDefinitions [{
      sphereCollisionRadius=0
      boundBone {boneName=dyn_braid_l_01}
      boxExtents=<10 15 15>
      constraintSetup {angularLimitsMax=<1 0 1>, angularLimitsMin=<-1 -15 -1>, angularTarget=<1 1 1>}
      localJointOffset=<2 0 0>
     }]
    relativeSpaceBone {boneName=spine_05}
   }
   Anim Dynamics - Bone: dyn_braid_l_02(ComponentPose=@n1, Alpha=1, ExternalForce="0,0,0") {
    angularDampingOverride=0.998
    angularSpringConstant=150
    bAngularSpring=true
    bGravityOverrideInSimSpace=true
    bOverrideAngularDamping=true
    bUseGravityOverride=true
    simulationSpace=RootRelative
    alphaBoolBlend {blendOption=Linear}
    boundBone {boneName=dyn_braid_l_02}
    componentAppliedLinearAccClamp=<3000 3000 3000>
    componentLinearAccScale=<0.3 0.3 0.3>
    componentLinearVelScale=<0.3 0.3 0.3>
    gravityOverride {z=-980}
    physicsBodyDefinitions [{
      sphereCollisionRadius=0
      boundBone {boneName=dyn_braid_l_02}
      boxExtents=<10 15 15>
      constraintSetup {angularLimitsMax=<1 0 1>, angularLimitsMin=<-1 -15 -1>, angularTarget=<1 1 1>}
      localJointOffset=<2 0 0>
     }]
    relativeSpaceBone {boneName=spine_05}
   }
   n2 = Anim Dynamics - Bone: dyn_braid_l_03(Alpha=1, ExternalForce="0,0,0") {
    angularDampingOverride=0.998
    angularSpringConstant=150
    bAngularSpring=true
    bGravityOverrideInSimSpace=true
    bOverrideAngularDamping=true
    bUseGravityOverride=true
    simulationSpace=RootRelative
    alphaBoolBlend {blendOption=Linear}
    boundBone {boneName=dyn_braid_l_03}
    componentAppliedLinearAccClamp=<3000 3000 3000>
    componentLinearAccScale=<0.3 0.3 0.3>
    componentLinearVelScale=<0.3 0.3 0.3>
    gravityOverride {z=-980}
    physicsBodyDefinitions [{
      sphereCollisionRadius=0
      boundBone {boneName=dyn_braid_l_03}
      boxExtents=<10 15 15>
      constraintSetup {angularLimitsMax=<1 0 15>, angularLimitsMin=<-1 -15 -1>, angularTarget=<1 1 1>}
      localJointOffset=<2 0 0>
     }]
    relativeSpaceBone {boneName=spine_05}
   }
   n3 = Anim Dynamics - Bone: dyn_braid_l_04(ComponentPose=@n2, Alpha=1, ExternalForce="0,0,0") {
    angularDampingOverride=0.998
    angularSpringConstant=150
    bAngularSpring=true
    bGravityOverrideInSimSpace=true
    bOverrideAngularDamping=true
    bUseGravityOverride=true
    simulationSpace=RootRelative
    alphaBoolBlend {blendOption=Linear}
    boundBone {boneName=dyn_braid_l_04}
    componentAppliedLinearAccClamp=<3000 3000 3000>
    componentLinearAccScale=<0.3 0.3 0.3>
    componentLinearVelScale=<0.3 0.3 0.3>
    gravityOverride {z=-980}
    physicsBodyDefinitions [{
      collisionType=InnerSphere
      sphereCollisionRadius=1.2674
      boundBone {boneName=dyn_braid_l_04}
      boxExtents=<10 15 15>
      constraintSetup {angularLimitsMax=<1 0 20>, angularLimitsMin=<-1 -15 -5>, angularTarget=<1 1 1>}
      localJointOffset=<2 0 0>
     }]
    relativeSpaceBone {boneName=spine_05}
    sphericalLimits [{limitRadius=10, drivingBone {boneName=spine_05}, sphereLocalOffset=<0 0 0>}]
   }
   Anim Dynamics - Bone: dyn_braid_l_05(ComponentPose=@n3, Alpha=1, ExternalForce="0,0,0") {
    angularDampingOverride=0.998
    angularSpringConstant=150
    bAngularSpring=true
    bGravityOverrideInSimSpace=true
    bOverrideAngularDamping=true
    bUseGravityOverride=true
    simulationSpace=RootRelative
    alphaBoolBlend {blendOption=Linear}
    boundBone {boneName=dyn_braid_l_05}
    componentAppliedLinearAccClamp=<3000 3000 3000>
    componentLinearAccScale=<0.3 0.3 0.3>
    componentLinearVelScale=<0.3 0.3 0.3>
    gravityOverride {z=-980}
    physicsBodyDefinitions [{
      sphereCollisionRadius=0
      boundBone {boneName=dyn_braid_l_05}
      boxExtents=<10 15 15>
      constraintSetup {angularLimitsMax=<1 0 30>, angularLimitsMin=<-1 -15 -20>, angularTarget=<1 1 1>}
      localJointOffset=<2 0 0>
     }]
    relativeSpaceBone {boneName=spine_05}
   }
   "Constraint - dyn_braid_l_05 (P)"(Alpha=1, ConstraintWeights_0=1) {
    alphaBoolBlend {blendOption=Linear}
    boneToModify {boneName=dyn_braid_l_05}
    constraintSetup [{targetBone {boneName=dyn_spine_04_l_lookat}}]
    constraintWeights [1]
   }
   n4 = Look At - Bone: dyn_braid_r_03(Alpha=1) {
    alphaBoolBlend {blendOption=Linear}
    boneToModify {boneName=dyn_braid_r_03}
    lookAtLocation=<0 3 -3>
    lookAtTarget {boneReference {boneName=dyn_spine_04_r_lookat}}
    lookAt_Axis {axis=<1 0>}
   }
   Look At - Bone: dyn_braid_l_03(ComponentPose=@n4, Alpha=1) {
    alphaBoolBlend {blendOption=Linear}
    boneToModify {boneName=dyn_braid_l_03}
    lookAtLocation=<0 3 2>
    lookAtTarget {boneReference {boneName=dyn_spine_04_l_lookat}}
    lookAt_Axis {axis=<1 0>}
   }
   Look At - Bone: dyn_braid_l_02(Alpha=1) {
    alphaBoolBlend {blendOption=Linear}
    boneToModify {boneName=dyn_braid_l_02}
    lookAtLocation=<0 3>
    lookAtTarget {boneReference {boneName=dyn_spine_04_l_lookat}}
    lookAt_Axis {axis=<1 0>}
   }
   Look At - Bone: dyn_braid_r_02(Alpha=1) {
    alphaBoolBlend {blendOption=Linear}
    boneToModify {boneName=dyn_braid_r_02}
    lookAtLocation {x=0}
    lookAtTarget {boneReference {boneName=dyn_spine_04_r_lookat}}
    lookAt_Axis {axis=<1 0>}
   }
   "Transform (Modify) Bone - Bone: dyn_braid_l_03"(Alpha=1, Translation="0,0,0", Rotation="0,0,0", Scale="1,1,1") {rotationMode=BMM_Additive, alphaBoolBlend {blendOption=Linear}, boneToModify {boneName=dyn_braid_l_03}}
   n5 = "Transform (Modify) Bone - Bone: dyn_collar"(
    ComponentPose=Local To Component(LocalPose=Copy Pose From Mesh() {bUseAttachedParent=true})
    Alpha=1
    Translation="0,0,0"
    Rotation="0,0,0"
    Scale="1,1.3,1.3"
   ) {scaleMode=BMM_Additive, alphaBoolBlend {blendOption=Linear}, boneToModify {boneName=dyn_collar}, scale {y=1.3, z=1.3}}
   n6 = Look At - Bone: dyn_braid_r_MSTR(ComponentPose=@n5, Alpha=0.7) {
    alphaBoolBlend {blendOption=Linear}
    boneToModify {boneName=dyn_braid_r_MSTR}
    lookAtLocation=<0 -3 2>
    lookAtTarget {boneReference {boneName=dyn_spine_04_r_lookat}}
    lookAt_Axis {axis=<1 0>}
   }
   n7 = Look At - Bone: dyn_braid_l_MSTR(ComponentPose=@n6, Alpha=0.7) {
    alphaBoolBlend {blendOption=Linear}
    boneToModify {boneName=dyn_braid_l_MSTR}
    lookAtLocation=<0 -2 -1>
    lookAtTarget {boneReference {boneName=dyn_spine_04_l_lookat}}
    lookAt_Axis {axis=<1 0>}
   }
   n8 = RigidBody(ComponentPose=@n7) {
    bClampLinearTranslationLimitToRefPose=true
    bOverrideWorldGravity=true
    overridePhysicsAsset=&/F_MED_Pliant_FaceAcc_Physics
    simulationSpace=BaseBoneSpace
    alphaBoolBlend {blendOption=Linear}
    baseBoneRef {boneName=neck_01}
    componentAppliedLinearAccClamp=<500 500 500>
    componentLinearAccScale=<0.1 0.1 0.1>
    componentLinearVelScale=<0.1 0.1 0.1>
    overrideWorldGravity {z=-980}
    simSpaceSettings {maxAngularAcceleration=200, maxAngularVelocity=10, maxLinearAcceleration=1000, maxLinearVelocity=50, worldAlpha=1, externalLinearDragV=<1 1>}
   }
   n9 = "Constraint - dyn_braid_r_05 (P)"(ComponentPose=@n8, Alpha=0.5, ConstraintWeights_0=1) {
    alphaBoolBlend {blendOption=Linear}
    boneToModify {boneName=dyn_braid_r_05}
    constraintSetup [{targetBone {boneName=dyn_spine_04_r_lookat}}]
    constraintWeights [1]
   }
   n10 = "Constraint - dyn_braid_l_05 (P)"(ComponentPose=@n9, Alpha=0.3, ConstraintWeights_0=1) {
    alphaBoolBlend {blendOption=Linear}
    boneToModify {boneName=dyn_braid_l_05}
    constraintSetup [{targetBone {boneName=dyn_spine_04_l_lookat}}]
    constraintWeights [1]
   }
   n11 = Control Rig(Source=Component To Local(ComponentPose=@n10)) {
    alphaBoolBlend {blendOption=Linear}
    controlRigAssetReference {blueprintRigClass=&/F_MED_Pliant_FaceAcc_CtrlRig.F_MED_Pliant_FaceAcc_CtrlRig_C}
   }
   Output Pose(Result=@n11) {layerGroup=None}
  ]}
 EventGraph {nodes [Event BlueprintUpdateAnimation()]}
]
variables []
